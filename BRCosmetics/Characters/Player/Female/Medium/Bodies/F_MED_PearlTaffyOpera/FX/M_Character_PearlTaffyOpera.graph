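# Masked, unlit Niagara sprite material that draws symbol textures for the PearlTaffyOpera character's FX.
# Particle color tints the symbols, then a dynamic-parameter hue shift and MF_FX_EmissiveAdjustments produce the emissive color.
# Texture and particle alpha multiply with a hard-edged dissolve, thresholding tiled Perlin noise by the Dissolve dynamic parameter and bypassed at full value.
# A separate Dissolve scalar parameter is left unconnected.
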Material /BRCosmetics/Characters/Player/Female/Medium/Bodies/F_MED_PearlTaffyOpera/FX/M_Character_PearlTaffyOpera
graph {
 properties {
  bAutomaticallySetUsageInEditor=false
  bCastRayTracedShadows=false
  bUsedWithNiagaraSprites=true
  blendMode=BLEND_Masked
  shadingModel=MSM_Unlit
  usedShadingModels=Unlit
 }
 nodes [
  n1 = DynamicParameter() {defaultValue=<rgba 1 0 1 1>, paramNames [Dissolve,HueShift,Param3,Param4]}
  n2 = ParticleColor()
  n3 = TextureSampleParameter2D() {parameterName=Texture, texture=&/T_PearlTaffyOpera_Symbols}
  MaterialFunctionCall(Emissive=MaterialFunctionCall(Hue Shift Percentage=@n1.HueShift, Texture=Multiply(A=@n2.RGB, B=@n3.RGB)) {materialFunction=&HueShift}.Result) {materialFunction=&MF_FX_EmissiveAdjustments}
  n4 = Constant() {r=1}
  n5 = Add(A=OneMinus(None=@n1.Dissolve)) {constB=0.015}
  n6 = TextureCoordinate() {uTiling=4, vTiling=4}
  n7 = SmoothStep(
   Min=@n5
   Max=@n5
   Value=LinearInterpolate(Alpha=TextureSample(UVs=@n6) {texture=/Game/Athena/HUD/XP/Materials/T_UI_PerlinNoiseRift}.RGB) {constA=0.1}
  )
  Multiply(A=@n2.A, B=Multiply(A=@n3.A, B=If(A=@n1.Dissolve, B=@n4, "A > B"=@n4, "A == B"=@n4, "A < B"=@n7)))
  ScalarParameter() {defaultValue=0.1, parameterName=Dissolve}
 ]
}
