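# Cosmetic Niagara effect for the PearlTaffyOpera character body, adding pulsing blue skirt glows, sparkles and drifting animated symbols.
# Glow emitters pin partitioned sprites to skirt bones, while Sparkles and two Symbols emitters spawn from skeletal mesh samples, including upper arms.
# Partitions pick each symbol's colour and flipbook frames, and inherited movement scales with owner speed.
# System timelines drive spawn probability and opacity, combined with distance-based fading on sparkles and symbols.
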NiagaraSystem /BRCosmetics/Characters/Player/Female/Medium/Bodies/F_MED_PearlTaffyOpera/FX/NS_Character_PearlTaffyOpera
dependencies {
 usedDataInterfaces [
  &NiagaraDataInterfaceArrayFloat
  &NiagaraDataInterfaceSpriteRendererInfo
  &NiagaraDataInterfaceCurve
  &NiagaraDataInterfaceArrayFloat2
  &NiagaraDataInterfaceVector2DCurve
  &NiagaraDataInterfaceSkeletalMesh
 ]
 usedDynamicInputs [
  &RandomRangeLinearColor
  &SelectVector2DFromArray
  &ReturnNormalizedExecIndex
  &Boolean_OrOperation
  &MakeFloatFromVector
 ]
 usedModules [
  &self:SystemSpawnScript.SystemScriptSource.SystemScriptGraph.NiagaraNodeAssignment_0.SetVariables_#1
  &SystemState
  &Timeline
 ]
}
emitters {
 Glow {
  emitterUpdateScript [
   &EmitterState(
    Inactive Response=::NewEnumerator1
    Life Cycle Mode=::NewEnumerator1
    Loop Behavior=::NewEnumerator1
    Loop Duration Mode=::NewEnumerator1
    UseLoopDelay=false
   )
   &SpawnBurst_Instantaneous(Spawn Count=8, Use Spawn Probability=false, UseLoopCountLimit=false)
  ]
  particleSpawnScript [
   &PartitionParticles(Number Of Partitions=8)
   &InitializeParticle(
    Color=<rgba 0 0.4755 1 0.1>
    Lifetime=0.5
    Sprite Size Mode=::NewEnumerator3
    Uniform Sprite Size=40
    UsePositionOffset=false
    Write Material Random=true
   ) {module=InitializeParticle001}
  ]
  particleUpdateScript [
   &ParticleState(Kill Particles When Lifetime Has Elapsed=false)
   /Niagara/Noise/Waveform(
    EnableExponent1=false
    GlobalClampMax=false
    GlobalClampMin=false
    UseInput1=false
    UsePhase1=false
    "Waveform [1]":ENiagara_Waveforms=::NewEnumerator2
    "[1] Amplitude Min/Max":Vector2f=<0.5 1>
    "[1] Compound Frequency"=0.1
    "[1] Frequency"=0.75
    "[1] Phase Scale":NiagaraFloat=$System.Timeline001.NormalizedTime
   )
   &SkeletalMeshLocation(
    Bone / Socket Index=$Particles.PartitionParticles.Partition
    Bone Sampling Mode=::NewEnumerator4
    MaskSpawnGroup=false
    Orientation Sampling=::NewEnumerator1
    Sampled Position Offset=&MergeHelper_VectorSwitch(
     0=<5 0 5>
     1=<5 0 5>
     2=<5 0 5>
     3=<5 0 5>
     4=<5 0 5>
     5=<5 0 5>
     6=<5 0 5>
     7=<5 0 5>
     Index=$Particles.PartitionParticles.Partition
     Num=::NewEnumerator6
    )
    Skeletal Mesh=DataInterface() {propertyValues {
      DefaultMesh=null
      ExcludeBoneName=root
      OptionalComponentClass=null
      PreviewMesh=/BRCosmetics/Characters/Player/Female/Medium/Bodies/F_MED_PearlTaffyOpera/Meshes/F_MED_PearlTaffyOpera
      SkinningMode=SkinOnTheFly
      SoftSourceActor=null
      SourceMode=Default
      UvSetIndex=0
      WholeMeshLOD=-1
      bExcludeBone=false
      bReadDeformedGeometry=true
      bRequireCurrentFrameData=true
      ComponentTags []
      FilteredBones [
       dyn_skirt_fr_03
       dyn_skirt_fr_r_03
       dyn_skirt_side_r_03
       dyn_skirt_bk_r_03
       dyn_skirt_bk_03
       dyn_skirt_bk_l_03
       dyn_skirt_side_l_03
       dyn_skirt_fr_l_03
      ]
      FilteredSockets []
      MeshUserParameter {parameter {name=None, typeDefHandle {registeredTypeIndex=1}, varData []}}
      SamplingRegions []
     }}
   )
   &ScaleSpriteSize(Uniform Curve Sprite Scale=DataInterface() {propertyValues {Curve {keys [
        {arriveTangent=1.1179, interpMode=RCIM_Cubic, leaveTangent=1.1179, tangentMode=RCTM_User, time=0, value=0.6}
        {arriveTangent=0.0224, interpMode=RCIM_Cubic, leaveTangent=0.0224, tangentMode=RCTM_User, time=1, value=1}
       ]}}})
   &CameraOffset(Camera Offset Amount=5)
   &ScaleColor(
    Scale Alpha=&Multiply_Float(A=&Lerp_Float(A=0.5, B=1, Alpha=$Particles.Waveform.Waveform), B=$System.Timeline001.NormalizedTime)
    Scale RGB=&Lerp_Vector(A=<0.5 0.5 0.5>, B=<1 1 1>, Alpha=$Particles.Waveform.Waveform)
    ScaleA=true
    ScaleRGB=true
   )
   &DynamicMaterialParameters(
    Index 0 Param 1=2
    Index 0 Param 2=1
    Index 0 Param 3=1
    Index 0 Param 4=1
    Module.Param0WriteEnabled=true
    Module.Param1WriteEnabled=true
    Module.Param2WriteEnabled=true
    Module.Param3WriteEnabled=true
   )
  ]
  rendererClasses [&NiagaraSpriteRendererProperties]
  renderers [{rendererClass=&NiagaraSpriteRendererProperties}]
 }
 Glow_Outline {
  emitterUpdateScript [
   &EmitterState(
    Inactive Response=::NewEnumerator1
    Life Cycle Mode=::NewEnumerator1
    Loop Behavior=::NewEnumerator1
    Loop Duration Mode=::NewEnumerator1
    UseLoopDelay=false
   )
   &SpawnBurst_Instantaneous(Spawn Count=10, Use Spawn Probability=true, UseLoopCountLimit=true)
  ]
  particleSpawnScript [
   &PartitionParticles(Number Of Partitions=10)
   &InitializeParticle(Color=<rgba 0 0.6667 1 0.5>, Sprite Size Mode=::NewEnumerator3, Uniform Sprite Size=20, UsePositionOffset=false, Write Material Random=true)
   &CameraOffset(Camera Offset Amount=20)
  ]
  particleUpdateScript [
   &ParticleState(Kill Particles When Lifetime Has Elapsed=false, Let Infinitely Lived Particles Die When Emitter Deactivates=true)
   &Waveform(
    EnableExponent1=false
    GlobalClampMax=false
    GlobalClampMin=false
    UseInput1=false
    UsePhase1=false
    "Waveform [1]":ENiagara_Waveforms=::NewEnumerator2
    "[1] Amplitude Min/Max":Vector2f=<0.5 1>
    "[1] Compound Frequency"=0.1
    "[1] Frequency"=0.75
    "[1] Phase Scale":NiagaraFloat=$System.Timeline001.NormalizedTime
   )
   &SkeletalMeshLocation(
    Bone / Socket Index=$Particles.PartitionParticles.Partition
    Bone Sampling Mode=::NewEnumerator4
    MaskSpawnGroup=false
    Orientation Sampling=::NewEnumerator1
    Sampled Position Offset=<5 0 0>
    Skeletal Mesh=DataInterface() {propertyValues {
      DefaultMesh=null
      ExcludeBoneName=root
      OptionalComponentClass=null
      PreviewMesh=&F_MED_PearlTaffyOpera
      SkinningMode=SkinOnTheFly
      SoftSourceActor=null
      SourceMode=Default
      UvSetIndex=0
      WholeMeshLOD=-1
      bExcludeBone=false
      bReadDeformedGeometry=true
      bRequireCurrentFrameData=true
      ComponentTags []
      FilteredBones [
       dyn_skirt_fr_03
       dyn_skirt_fr_r_03
       dyn_skirt_side_r_03
       dyn_skirt_bk_r_03
       dyn_skirt_bk_03
       dyn_skirt_bk_l_03
       dyn_skirt_side_l_03
       dyn_skirt_fr_l_03
      ]
      FilteredSockets []
      MeshUserParameter {parameter {name=None, typeDefHandle {registeredTypeIndex=1}, varData []}}
      SamplingRegions []
     }}
   )
   &ScaleColor(
    Scale Alpha=$System.Timeline001.NormalizedTime
    Scale RGB=&Lerp_Vector(A=<0.5 0.5 0.5>, B=<1 1 1>, Alpha=$Particles.Waveform.Waveform)
    ScaleA=true
    ScaleRGB=true
   )
  ]
  rendererClasses [&NiagaraSpriteRendererProperties]
  renderers [{rendererClass=&NiagaraSpriteRendererProperties}]
 }
 Sparkles {
  emitterUpdateScript [
   &EmitterState()
   &SpawnRate(Spawn Probability=$System.Timeline001.NormalizedTime, SpawnRate=1.5, Use Spawn Probability=true)
  ]
  particleSpawnScript [
   &InitializeParticle(
    Lifetime=0.5
    Position Offset=<0 5 -3>
    Position Offset Coordinate Space=::Local
    Sprite Rotation Mode=::NewEnumerator2
    Sprite Size Mode=::NewEnumerator3
    Uniform Sprite Size=&RandomRangeFloat(Minimum=5, Maximum=7.5)
    UsePositionOffset=true
    Write Material Random=true
   )
   &SkeletalMeshLocation(
    Bone / Socket Index=&RandomRangeInt(Minimum=0, Maximum=1)
    Bone Sampling Mode=::NewEnumerator4
    MaskSpawnGroup=false
    Skeletal Mesh=$System.Skeletal Mesh Reader
   )
   /SetVariables(Particles.RandomParticle:NiagaraInt32=&RandomRangeInt(Minimum=0, Maximum=9)) {isSetParameters=true}
   &ShapeLocation(
    MaskSpawnGroup=false
    Non Uniform Scale=<0.8 1 1>
    Offset=<0 0 -25>
    Offset Mode=::NewEnumerator0
    Override Seed=false
    Rotation Angle=-0.15
    Rotation Angle Type:ENiagara_AngleInput=::NewEnumerator0
    Rotation Axis=<0 1 0>
    Rotation Mode=::NewEnumerator1
    Sphere Radius=20
    Sphere Surface Distribution=1
    UseOwnerScale=true
   )
   &CameraOffset(Camera Offset Amount=50)
  ]
  particleUpdateScript [
   /SetVariables(Particles.DistanceScale:NiagaraFloat=&Multiply_Float(
     A=&NormalizeDistanceRange(
      Start Position:NiagaraPosition=&AddVectorToPosition(Position=/Niagara/DynamicInputs/Helpers/SimulationPosition(), Vector=<0 0 50>)
      End Position:NiagaraPosition=$Particles.Position
      Distance:NiagaraFloat=500
      Invert Normalized Range:NiagaraBool=true
     )
     B=$System.Timeline.InverseNormalizedTime
    )) {isSetParameters=true}
   &ParticleState()
   &SubUVAnimation(
    End Frame Range Override=3
    "Sprite Renderer "=DataInterface() {propertyValues {SpriteRenderer=&self:Sparkles_10.NiagaraSpriteRendererProperties_0}}
    UseEndFrame=true
    UseStartFrame=true
   )
   &DynamicMaterialParameters(
    Index 0 Param 1=&FloatFromCurve(FloatCurve=DataInterface() {propertyValues {Curve {keys [
         {arriveTangent=-0.1162, interpMode=RCIM_Cubic, leaveTangent=-0.1162, tangentMode=RCTM_User, time=0, value=1}
         {arriveTangent=-0.0508, interpMode=RCIM_Cubic, leaveTangent=-0.0508, tangentMode=RCTM_User, time=0.7717, value=1}
         {tangentMode=RCTM_User, time=1, value=0}
        ]}}})
    Module.Index0Param0Write=true
    Module.Index0Param1Write=false
    Module.Index0Param2Write=false
    Module.Index0Param3Write=false
   )
   &InheritSourceMovement(
    Applied Position Scale=&Multiply_VectorByFloat(
     Vector=<0.95 0.95 0.95>
     Float=&ClampFloat(Float=&RemapRange(
       Input Value To Remap=&VectorLength(Vector=$Engine.Owner.Velocity)
       Input Min=50
       Input Max=500
       Output Min=0
       Output Max=1
      ))
    )
    ApplyForce=false
    ApplyPosition=true
    ApplyPositionFalloff=false
   )
   &ScaleColor(Scale Alpha=$Particles.DistanceScale, ScaleA=true, ScaleRGB=true)
   &SolveForcesAndVelocity(Clamp Velocity=false, Limit Acceleration=false)
   &KillParticles(Kill Particles=&SetBoolByFloatComparison(
     A=&MakeFloatFromVector4(Vector 4=&MakeVector4_FromLinearColor(LinearColor=$Particles.Color), Channel=::NewEnumerator3)
     B=0.01
     Comparison Type=::NewEnumerator7
    ))
  ]
  rendererClasses [&NiagaraSpriteRendererProperties]
  renderers [{rendererClass=&NiagaraSpriteRendererProperties}]
 }
 Symbols {
  emitterUpdateScript [
   &EmitterState()
   &SpawnRate(Spawn Probability=$System.Timeline001.NormalizedTime, SpawnRate=10, Use Spawn Probability=true)
  ]
  particleSpawnScript [
   &PartitionParticles(Number Of Partitions=8, Partition Mode:ENiagara_PartitionMode=::NewEnumerator5)
   &InitializeParticle(
    Color=&MergeHelper_ColorSwitch(
     0=<rgba 1 1 1 1>
     1=<rgba 1 1 1 1>
     2=<rgba 0.1787 0.0591 2 1>
     3=<rgba 0 2 1.3368 1>
     4=<rgba 2 0.1074 0.1523 1>
     5=<rgba 1 1 1 1>
     6=<rgba 2 0.1 1.3667 1>
     7=<rgba 1 1 1 1>
     Index=$Particles.PartitionParticles.Partition
     Num=::NewEnumerator6
    )
    Lifetime Max=1.5
    Lifetime Mode=::NewEnumerator1
    Sprite Rotation Mode=::NewEnumerator2
    Sprite Size Mode=::NewEnumerator1
    Uniform Sprite Size Max=4
    Uniform Sprite Size Min=3
    UsePositionOffset=true
    Write Material Random=true
   )
   &SkeletalMeshLocation(Bone Sampling Mode=::NewEnumerator0, MaskSpawnGroup=false, OverrideRandomSeed=false, Skeletal Mesh=$System.Skeletal Mesh Reader)
   &ShapeLocation(
    MaskSpawnGroup=false
    Non Uniform Scale=<1 1 0>
    Offset=<0 0 -10>
    Offset Mode=::NewEnumerator0
    Override Seed=false
    Rotation Angle Type:ENiagara_AngleInput=::NewEnumerator0
    Rotation Axis=<0 1 0>
    Rotation Mode=::NewEnumerator1
    Sphere Radius=20
    Sphere Surface Distribution=1
    UseOwnerScale=true
   )
   &CameraOffset(Camera Offset Amount=20)
   &AddVelocity(
    Module.Use Falloff Distance=false
    Origin Offset=<0 0 50>
    Override Seed=false
    Velocity Mode=::NewEnumerator1
    Velocity Origin=$Particles.SkeletalMeshLocation.SampledPosition
    Velocity Speed=100
   )
  ]
  particleUpdateScript [
   /SetVariables(Particles.DistanceScale:NiagaraFloat=&Multiply_Float(
     A=&FloatFromCurve(
      FloatCurve=DataInterface() {propertyValues {Curve {
         defaultValue=0
         keys [
          {interpMode=RCIM_Cubic, tangentMode=RCTM_User, time=0, value=0}
          {arriveTangent=3.1086, interpMode=RCIM_Cubic, leaveTangent=3.1086, tangentMode=RCTM_User, time=1, value=1}
         ]
        }}}
      CurveIndex=&NormalizeDistanceRange(
       Start Position:NiagaraPosition=$Particles.Initial.SkeletalMeshLocation.SampledPosition
       End Position:NiagaraPosition=$Particles.Position
       Distance:NiagaraFloat=500
       Invert Normalized Range:NiagaraBool=true
      )
     )
     B=$System.Timeline.InverseNormalizedTime
    )) {isSetParameters=true, module=SetVariables_#2}
   &ParticleState()
   &ScaleSpriteSize(
    Non-Uniform Curve Scale=&Vector2DFromFloat(Value=$Particles.DistanceScale)
    Non-Uniform Curve Sprite Scale=DataInterface() {propertyValues {
      XCurve {
       defaultValue=0
       keys [
        {arriveTangent=27.0378, interpMode=RCIM_Cubic, leaveTangent=27.0378, tangentMode=RCTM_User, time=0, value=0}
        {
         arriveTangent=-6.4665
         arriveTangentWeight=0.0469
         interpMode=RCIM_Cubic
         leaveTangent=-6.4665
         tangentMode=RCTM_User
         time=0.0289
         value=1.195
        }
        {arriveTangent=-0.7599, interpMode=RCIM_Cubic, leaveTangent=-0.7599, time=0.0957, value=0.5946}
        {arriveTangent=1.0763, interpMode=RCIM_Cubic, leaveTangent=1.0763, time=0.2498, value=1.0271}
        {arriveTangent=-0.484, interpMode=RCIM_Cubic, leaveTangent=-0.484, time=0.5511, value=1.0847}
        {arriveTangent=-2.4165, interpMode=RCIM_Cubic, leaveTangent=-2.4165, time=0.8845, value=0.72}
        {arriveTangent=-0.0405, arriveTangentWeight=0.2864, interpMode=RCIM_Cubic, leaveTangent=-0.0405, tangentMode=RCTM_User, time=1, value=0}
       ]
      }
      YCurve {
       defaultValue=0
       keys [
        {arriveTangent=6.4321, interpMode=RCIM_Cubic, leaveTangent=6.4321, tangentMode=RCTM_User, time=0, value=0}
        {arriveTangent=10.7376, interpMode=RCIM_Cubic, leaveTangent=10.7376, time=0.0648, value=0.9991}
        {arriveTangent=-0.5061, interpMode=RCIM_Cubic, leaveTangent=-0.5061, tangentMode=RCTM_User, time=0.1399, value=1.5024}
        {arriveTangent=-1.4131, interpMode=RCIM_Cubic, leaveTangent=-1.4131, tangentMode=RCTM_User, time=0.6265, value=0.8211}
        {arriveTangent=-4.8374, interpMode=RCIM_Cubic, leaveTangent=-4.8374, tangentMode=RCTM_User, time=0.839, value=0.869}
        {arriveTangent=-0.0405, interpMode=RCIM_Cubic, leaveTangent=-0.0405, tangentMode=RCTM_User, time=1, value=0}
       ]
      }
     }}
    Scale Sprite Size Mode=::NewEnumerator3
   )
   &SubUVAnimation(
    End Frame Range Override:NiagaraInt32=&MergeHelper_IntSwitch(0=2, 1=4, 2=6, 3=8, 4=10, 5=12, 6=14, 7=16, Index=$Particles.PartitionParticles.Partition, Num=::NewEnumerator6)
    Play Rate=3
    Random Start Frame=true
    "Sprite Renderer "=DataInterface() {propertyValues {SpriteRenderer=&self:Symbols_0.NiagaraSpriteRendererProperties_0}}
    Start Frame Range Override=&MergeHelper_IntSwitch(0=0, 1=2, 2=4, 3=6, 4=8, 5=10, 6=12, 7=14, Index=$Particles.PartitionParticles.Partition, Num=::NewEnumerator6)
    SubUV Animation Mode:ENiagara_SubUVLookupModeV2=::NewEnumerator3
    UseEndFrame=true
    UsePlayRateScale=false
    UseStartFrame=true
   )
   &DynamicMaterialParameters(
    Index 0 Param 1=&FloatFromCurve(FloatCurve=DataInterface() {propertyValues {Curve {keys [
         {interpMode=RCIM_Cubic, tangentMode=RCTM_User, time=0, value=0}
         {arriveTangent=-0.1162, interpMode=RCIM_Cubic, leaveTangent=-0.1162, tangentMode=RCTM_User, time=0.2, value=1}
         {arriveTangent=-0.0508, interpMode=RCIM_Cubic, leaveTangent=-0.0508, tangentMode=RCTM_User, time=0.7717, value=1}
         {tangentMode=RCTM_User, time=1, value=0}
        ]}}})
    Index 0 Param 2=&MergeHelper_FloatSwitch(
     0=&SelectFloatFromArray(
      Array Sampling Mode=::NewEnumerator0
      Direct Array Index=?
      Float Selection Array=DataInterface() {propertyValues {FloatData [0,0]}}
      Random Seed=0
     )
     1=0
     2=0
     3=0
     4=0
     5=&RandomRangeFloat(Minimum=0, Maximum=1)
     6=0
     7=&RandomRangeFloat(Minimum=0, Maximum=1)
     Index=$Particles.PartitionParticles.Partition
     Num=::NewEnumerator6
    )
    Module.Index0Param0Write=true
    Module.Index0Param1Write=true
    Module.Index0Param2Write=false
    Module.Index0Param3Write=false
   )
   &InheritSourceMovement(
    Applied Position Scale=&Multiply_VectorByFloat(
     Vector=<0.95 0.95 0.95>
     Float=&ClampFloat(Float=&RemapRange(
       Input Value To Remap=&VectorLength(Vector=$Engine.Owner.Velocity)
       Input Min=50
       Input Max=500
       Output Min=0
       Output Max=1
      ))
    )
    ApplyForce=false
    ApplyPosition=true
    ApplyPositionFalloff=false
   )
   &AccelerationForce(Acceleration=&MakeVector(
     X=0
     Y=0
     Z=&FloatFromCurve(
      FloatCurve=DataInterface() {propertyValues {Curve {
         defaultValue=1
         keys [{interpMode=RCIM_Cubic, tangentMode=RCTM_User, time=0, value=0}, {interpMode=RCIM_Cubic, tangentMode=RCTM_User, time=1, value=1}]
        }}}
      Scale Curve=100
     )
    ))
   &PointForce(
    Force Origin=&AddVectorToPosition(Position=&SimulationPosition(), Vector=<0 0 50>)
    Force Strength=200
    Module.Use Falloff=false
    Module.UseLinearForce=false
   )
   &WindForce(
    Offset By Time:Vector3f=&RandomRangeVector(Minimum=<0 0 -5>, Maximum=<0 0 5>)
    Wind Speed=&RandomRangeVector(Minimum=<0 0 2.5>, Maximum=<0 0 10>)
   )
   &Drag(Drag=&RandomRangeFloat(Minimum=10, Maximum=20), Use Linear Drag=true, Use Rotational Drag=true)
   &SolveForcesAndVelocity(Clamp Velocity=false, Limit Acceleration=false)
   &ScaleColor(Scale Alpha=$Particles.DistanceScale, ScaleA=true, ScaleRGB=true)
   &KillParticles(Kill Particles=&SetBoolByFloatComparison(
     A=&MakeFloatFromVector4(Vector 4=&MakeVector4_FromLinearColor(LinearColor=$Particles.Color), Channel=::NewEnumerator3)
     B=0.01
     Comparison Type=::NewEnumerator7
    ))
   &JitterPosition(Jitter Amount=0.2)
  ]
  rendererClasses [&NiagaraSpriteRendererProperties]
  renderers [{rendererClass=&NiagaraSpriteRendererProperties}]
 }
 SymbolsUpperArm {
  emitterUpdateScript [
   &EmitterState()
   &SpawnRate(Spawn Probability=$System.Timeline001.NormalizedTime, SpawnRate=3, Use Spawn Probability=true)
  ]
  particleSpawnScript [
   &PartitionParticles(Number Of Partitions=8)
   &InitializeParticle(
    Color=&MergeHelper_ColorSwitch(
     0=<rgba 1 1 1 1>
     1=<rgba 1 1 1 1>
     2=<rgba 0.1787 0.0591 2 1>
     3=<rgba 0 2 1.3368 1>
     4=<rgba 2 0.1074 0.1523 1>
     5=<rgba 1 1 1 1>
     6=<rgba 2 0.1 1.3667 1>
     7=<rgba 1 1 1 1>
     Index=$Particles.PartitionParticles.Partition
     Num=::NewEnumerator6
    )
    Lifetime Max=1.25
    Lifetime Min=0.75
    Lifetime Mode=::NewEnumerator1
    Sprite Rotation Mode=::NewEnumerator2
    Sprite Size Mode=::NewEnumerator1
    Uniform Sprite Size Max=2.5
    Uniform Sprite Size Min=1.5
    UsePositionOffset=false
    Write Material Random=true
   )
   &SkeletalMeshLocation(
    Bone Sampling Mode=::NewEnumerator0
    MaskSpawnGroup=false
    OverrideRandomSeed=false
    Skeletal Mesh=DataInterface() {propertyValues {
      DefaultMesh=null
      ExcludeBoneName=root
      OptionalComponentClass=null
      PreviewMesh=&F_MED_PearlTaffyOpera
      SkinningMode=SkinOnTheFly
      SoftSourceActor=null
      SourceMode=Default
      UvSetIndex=0
      WholeMeshLOD=-1
      bExcludeBone=false
      bReadDeformedGeometry=true
      bRequireCurrentFrameData=true
      ComponentTags []
      FilteredBones [upperarm_l,upperarm_r]
      FilteredSockets []
      MeshUserParameter {parameter {name=None, typeDefHandle {registeredTypeIndex=1}, varData []}}
      SamplingRegions []
     }}
   )
   &ShapeLocation(
    MaskSpawnGroup=false
    Non Uniform Scale=<1 1 0>
    Offset=<0 0 10>
    Offset Mode=::NewEnumerator0
    Override Seed=false
    Rotation Angle Type:ENiagara_AngleInput=::NewEnumerator0
    Rotation Axis=<0 1 0>
    Rotation Mode=::NewEnumerator1
    Sphere Radius=5
    Sphere Surface Distribution=1
    UseOwnerScale=true
   )
   &CameraOffset(Camera Offset Amount=20)
   &AddVelocity(
    Module.Use Falloff Distance=false
    Origin Offset=<0 0 50>
    Override Seed=false
    Velocity Mode=::NewEnumerator1
    Velocity Origin=$Particles.SkeletalMeshLocation.SampledPosition
    Velocity Speed=100
   )
  ]
  particleUpdateScript [
   /SetVariables(Particles.DistanceScale:NiagaraFloat=&Multiply_Float(
     A=&FloatFromCurve(
      FloatCurve=DataInterface() {propertyValues {Curve {
         defaultValue=0
         keys [
          {interpMode=RCIM_Cubic, tangentMode=RCTM_User, time=0, value=0}
          {arriveTangent=3.1086, interpMode=RCIM_Cubic, leaveTangent=3.1086, tangentMode=RCTM_User, time=1, value=1}
         ]
        }}}
      CurveIndex=&NormalizeDistanceRange(
       Start Position:NiagaraPosition=$Particles.Initial.SkeletalMeshLocation.SampledPosition
       End Position:NiagaraPosition=$Particles.Position
       Distance:NiagaraFloat=200
       Invert Normalized Range:NiagaraBool=true
      )
     )
     B=$System.Timeline.InverseNormalizedTime
    )) {isSetParameters=true, module=SetVariables_#2}
   &ParticleState()
   &ScaleSpriteSize(
    Non-Uniform Curve Scale=&Vector2DFromFloat(Value=$Particles.DistanceScale)
    Non-Uniform Curve Sprite Scale=DataInterface() {propertyValues {
      XCurve {
       defaultValue=0
       keys [
        {arriveTangent=27.0378, interpMode=RCIM_Cubic, leaveTangent=27.0378, tangentMode=RCTM_User, time=0, value=0}
        {
         arriveTangent=-6.4665
         arriveTangentWeight=0.0469
         interpMode=RCIM_Cubic
         leaveTangent=-6.4665
         tangentMode=RCTM_User
         time=0.0289
         value=1.195
        }
        {arriveTangent=-0.7599, interpMode=RCIM_Cubic, leaveTangent=-0.7599, time=0.0957, value=0.5946}
        {arriveTangent=1.0763, interpMode=RCIM_Cubic, leaveTangent=1.0763, time=0.2498, value=1.0271}
        {arriveTangent=-0.484, interpMode=RCIM_Cubic, leaveTangent=-0.484, time=0.5511, value=1.0847}
        {arriveTangent=-2.4165, interpMode=RCIM_Cubic, leaveTangent=-2.4165, time=0.8845, value=0.72}
        {arriveTangent=-0.0405, arriveTangentWeight=0.2864, interpMode=RCIM_Cubic, leaveTangent=-0.0405, tangentMode=RCTM_User, time=1, value=0}
       ]
      }
      YCurve {
       defaultValue=0
       keys [
        {arriveTangent=6.4321, interpMode=RCIM_Cubic, leaveTangent=6.4321, tangentMode=RCTM_User, time=0, value=0}
        {arriveTangent=10.7376, interpMode=RCIM_Cubic, leaveTangent=10.7376, time=0.0648, value=0.9991}
        {arriveTangent=-0.5061, interpMode=RCIM_Cubic, leaveTangent=-0.5061, tangentMode=RCTM_User, time=0.1399, value=1.5024}
        {arriveTangent=-1.4131, interpMode=RCIM_Cubic, leaveTangent=-1.4131, tangentMode=RCTM_User, time=0.6265, value=0.8211}
        {arriveTangent=-4.8374, interpMode=RCIM_Cubic, leaveTangent=-4.8374, tangentMode=RCTM_User, time=0.839, value=0.869}
        {arriveTangent=-0.0405, interpMode=RCIM_Cubic, leaveTangent=-0.0405, tangentMode=RCTM_User, time=1, value=0}
       ]
      }
     }}
    Scale Sprite Size Mode=::NewEnumerator3
   )
   &SubUVAnimation(
    End Frame Range Override:NiagaraInt32=&MergeHelper_IntSwitch(0=2, 1=4, 2=6, 3=8, 4=10, 5=12, 6=14, 7=16, Index=$Particles.PartitionParticles.Partition, Num=::NewEnumerator6)
    Play Rate=3
    Random Start Frame=true
    "Sprite Renderer "=DataInterface() {propertyValues {SpriteRenderer=&self:SymbolsUpperArm_0.NiagaraSpriteRendererProperties_0}}
    Start Frame Range Override=&MergeHelper_IntSwitch(0=0, 1=2, 2=4, 3=6, 4=8, 5=10, 6=12, 7=14, Index=$Particles.PartitionParticles.Partition, Num=::NewEnumerator6)
    SubUV Animation Mode:ENiagara_SubUVLookupModeV2=::NewEnumerator3
    UseEndFrame=true
    UsePlayRateScale=false
    UseStartFrame=true
   )
   &DynamicMaterialParameters(
    Index 0 Param 1=&FloatFromCurve(FloatCurve=DataInterface() {propertyValues {Curve {keys [
         {interpMode=RCIM_Cubic, tangentMode=RCTM_User, time=0, value=0}
         {arriveTangent=-0.1162, interpMode=RCIM_Cubic, leaveTangent=-0.1162, tangentMode=RCTM_User, time=0.2, value=1}
         {arriveTangent=-0.0508, interpMode=RCIM_Cubic, leaveTangent=-0.0508, tangentMode=RCTM_User, time=0.7717, value=1}
         {tangentMode=RCTM_User, time=1, value=0}
        ]}}})
    Index 0 Param 2=&MergeHelper_FloatSwitch(
     0=&SelectFloatFromArray(
      Array Sampling Mode=::NewEnumerator0
      Direct Array Index=?
      Float Selection Array=DataInterface() {propertyValues {FloatData [0,0]}}
      Random Seed=0
     )
     1=0
     2=0
     3=0
     4=0
     5=&RandomRangeFloat(Minimum=0, Maximum=1)
     6=0
     7=&RandomRangeFloat(Minimum=0, Maximum=1)
     Index=$Particles.PartitionParticles.Partition
     Num=::NewEnumerator6
    )
    Module.Index0Param0Write=true
    Module.Index0Param1Write=true
    Module.Index0Param2Write=false
    Module.Index0Param3Write=false
   )
   &InheritSourceMovement(
    Applied Position Scale=&Multiply_VectorByFloat(
     Vector=<0.95 0.95 0.95>
     Float=&ClampFloat(Float=&RemapRange(
       Input Value To Remap=&VectorLength(Vector=$Engine.Owner.Velocity)
       Input Min=50
       Input Max=500
       Output Min=0
       Output Max=1
      ))
    )
    ApplyForce=false
    ApplyPosition=true
    ApplyPositionFalloff=false
   )
   &AccelerationForce(Acceleration=&MakeVector(
     X=0
     Y=0
     Z=&FloatFromCurve(
      FloatCurve=DataInterface() {propertyValues {Curve {
         defaultValue=1
         keys [{interpMode=RCIM_Cubic, tangentMode=RCTM_User, time=0, value=0}, {interpMode=RCIM_Cubic, tangentMode=RCTM_User, time=1, value=1}]
        }}}
      Scale Curve=100
     )
    ))
   &PointForce(
    Force Origin=&AddVectorToPosition(Position=&SimulationPosition(), Vector=<0 0 120>)
    Force Strength=200
    Module.Use Falloff=false
    Module.UseLinearForce=false
   )
   &Drag(Drag=&RandomRangeFloat(Minimum=10, Maximum=15), Use Linear Drag=true, Use Rotational Drag=true)
   &SolveForcesAndVelocity(Clamp Velocity=false, Limit Acceleration=false)
   &ScaleColor(Scale Alpha=$Particles.DistanceScale, ScaleA=true, ScaleRGB=true)
   &KillParticles(Kill Particles=&SetBoolByFloatComparison(
     A=&MakeFloatFromVector4(Vector 4=&MakeVector4_FromLinearColor(LinearColor=$Particles.Color), Channel=::NewEnumerator3)
     B=0.01
     Comparison Type=::NewEnumerator7
    ))
   &JitterPosition(Jitter Amount=0.2)
  ]
  rendererClasses [&NiagaraSpriteRendererProperties]
  renderers [{rendererClass=&NiagaraSpriteRendererProperties}]
 }
}
summary {
 systemName=NS_Character_PearlTaffyOpera
 emitters [
  {emitterName=Sparkles, rendererClasses [&NiagaraSpriteRendererProperties]}
  {emitterName=Symbols, rendererClasses [&NiagaraSpriteRendererProperties]}
  {emitterName=Glow, rendererClasses [&NiagaraSpriteRendererProperties]}
  {emitterName=Glow_Outline, rendererClasses [&NiagaraSpriteRendererProperties]}
  {emitterName=SymbolsUpperArm, rendererClasses [&NiagaraSpriteRendererProperties]}
 ]
 userVariables [
  User.bIsFrontEnd {type=NiagaraBool}
  User.bIsFrontEndPreview {type=NiagaraBool}
  User.bisHeadHidden {type=NiagaraBool}
  User.EnableFrontendFakeAudioLoop {type=NiagaraFloat, defaultValue=0, description=""}
  User.GlowOn {type=NiagaraFloat, defaultValue=0, description=""}
 ]
}
