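# Control Rig for the PearlTaffyOpera face accessory that positions the back and side hair nulls (dyn_hair_*) to follow body pose.
# Weighted parent constraints blend the nulls between head, neck, spine, clavicle and upper-arm bones.
# Swing-twist rotation deltas, spherical pose readers and pelvis-height remaps produce pose weights such as Head Up and Crouch.
# These weights, often multiplied into combination poses, drive layered corrective offsets on the nulls.
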ControlRigBlueprint /BRCosmetics/Characters/Player/Female/Medium/Bodies/F_MED_PearlTaffyOpera/Meshes/Parts/F_MED_PearlTaffyOpera_FaceAcc_CtrlRig
graphs [
 Add_5_ContainedGraph {nodes [
   n1 = RigVMFunctionEntryNode()
   Add = &RigVMFunction_MathDoubleAdd(A=@n1.A, B=@n1.B)
   RigVMFunctionReturnNode(Result=&RigVMFunction_MathDoubleAdd(A=@Add.Result, B=@n1.C).Result)
  ]}
 Add_5_ContainedGraph {nodes [
   n1 = RigVMFunctionEntryNode()
   Add = &RigVMFunction_MathDoubleAdd(A=@n1.A, B=@n1.B)
   RigVMFunctionReturnNode(Result=&RigVMFunction_MathDoubleAdd(A=@Add.Result, B=@n1.C).Result)
  ]}
 CollapseNode_ContainedGraph {nodes [
   n1 = RigVMFunctionEntryNode()
   &RigUnit_ParentConstraint(
    AdvancedSettings=(InterpolationType=Average,RotationOrderForFilter=XZY)
    Child=(Type=Null,Name="dyn_hair_bk_l_01_null")
    Parents=((Item=(Type=Bone,Name="head"),Weight=1))
    Weight=1
    ExecutePin=@n1.ExecutePin
   )
   > &RigUnit_ParentConstraint(
    AdvancedSettings=(InterpolationType=Average,RotationOrderForFilter=XZY)
    Child=(Type=Null,Name="dyn_hair_bk_l_02_null")
    Parents="((Item=(Type=Bone,Name=\"clavicle_l\"),Weight=0.0503),(Item=(Type=Bone,Name=\"spine_04\"),Weight=0.0613),(Item=(Type=Bone,Name=\"head\"),Weight=0.1054),(Item=(Type=Bone,Name=\"neck_02\"),Weight=0.1567),(Item=(Type=Bone,Name=\"spine_05\"),Weight=0.2205),(Item=(Type=B ...(truncated)"
    Weight=1
   )
   > &RigUnit_ParentConstraint(
    AdvancedSettings=(InterpolationType=Average,RotationOrderForFilter=XZY)
    Child=(Type=Null,Name="dyn_hair_bk_l_03_null")
    Parents=((Item=(Type=Bone,Name="clavicle_l"),Weight=0.4559),(Item=(Type=Bone,Name="spine_05"),Weight=0.5441))
    Weight=1
   )
   > &RigUnit_ParentConstraint(
    AdvancedSettings=(InterpolationType=Average,RotationOrderForFilter=XZY)
    Child=(Type=Null,Name="dyn_hair_bk_l_04_null")
    Parents=((Item=(Type=Bone,Name="upperarm_l"),Weight=0.0282),(Item=(Type=Bone,Name="spine_04"),Weight=0.0451),(Item=(Type=Bone,Name="spine_05"),Weight=0.377),(Item=(Type=Bone,Name="clavicle_l"),Weight=0.5498))
    Weight=1
   )
   > &RigUnit_ParentConstraint(
    AdvancedSettings=(InterpolationType=Average,RotationOrderForFilter=XZY)
    Child=(Type=Null,Name="dyn_hair_bk_l_05_null")
    Parents=((Item=(Type=Bone,Name="upperarm_l"),Weight=0.0176),(Item=(Type=Bone,Name="spine_03"),Weight=0.0228),(Item=(Type=Bone,Name="spine_05"),Weight=0.285),(Item=(Type=Bone,Name="clavicle_l"),Weight=0.3331),(Item=(Type=Bone,Name="spine_04"),Weight=0.3415))
    Weight=1
   )
   > &RigUnit_ParentConstraint(
    AdvancedSettings=(InterpolationType=Average,RotationOrderForFilter=XZY)
    Child=(Type=Null,Name="dyn_hair_bk_l_06_null")
    Parents=((Item=(Type=Bone,Name="spine_05"),Weight=0.0187),(Item=(Type=Bone,Name="spine_02"),Weight=0.0811),(Item=(Type=Bone,Name="clavicle_l"),Weight=0.1267),(Item=(Type=Bone,Name="spine_04"),Weight=0.3575),(Item=(Type=Bone,Name="spine_03"),Weight=0.416))
    Weight=1
   )
   > &RigUnit_ParentConstraint(
    AdvancedSettings=(InterpolationType=Average,RotationOrderForFilter=XZY)
    Child=(Type=Null,Name="dyn_hair_bk_01_null")
    Parents=((Item=(Type=Bone,Name="neck_02"),Weight=0.0434),(Item=(Type=Bone,Name="head"),Weight=0.9566))
    Weight=1
   )
   > &RigUnit_ParentConstraint(
    AdvancedSettings=(InterpolationType=Average,RotationOrderForFilter=XZY)
    Child=(Type=Null,Name="dyn_hair_bk_02_null")
    Parents="((Item=(Type=Bone,Name=\"clavicle_r\"),Weight=0.015),(Item=(Type=Bone,Name=\"clavicle_l\"),Weight=0.015),(Item=(Type=Bone,Name=\"spine_04\"),Weight=0.0681),(Item=(Type=Bone,Name=\"head\"),Weight=0.0921),(Item=(Type=Bone,Name=\"neck_02\"),Weight=0.1321),(Item=(Type=B ...(truncated)"
    Weight=1
   )
   > &RigUnit_ParentConstraint(
    AdvancedSettings=(InterpolationType=Average,RotationOrderForFilter=XZY)
    Child=(Type=Null,Name="dyn_hair_bk_03_null")
    Parents=((Item=(Type=Bone,Name="clavicle_l"),Weight=0.0369),(Item=(Type=Bone,Name="clavicle_r"),Weight=0.0369),(Item=(Type=Bone,Name="spine_05"),Weight=0.9261))
    Weight=1
   )
   > &RigUnit_ParentConstraint(
    AdvancedSettings=(InterpolationType=Average,RotationOrderForFilter=XZY)
    Child=(Type=Null,Name="dyn_hair_bk_04_null")
    Parents=((Item=(Type=Bone,Name="spine_04"),Weight=0.0415),(Item=(Type=Bone,Name="clavicle_l"),Weight=0.0758),(Item=(Type=Bone,Name="clavicle_r"),Weight=0.0758),(Item=(Type=Bone,Name="spine_05"),Weight=0.807))
    Weight=1
   )
   > &RigUnit_ParentConstraint(
    AdvancedSettings=(InterpolationType=Average,RotationOrderForFilter=XZY)
    Child=(Type=Null,Name="dyn_hair_bk_05_null")
    Parents=((Item=(Type=Bone,Name="spine_03"),Weight=0.0792),(Item=(Type=Bone,Name="clavicle_r"),Weight=0.163),(Item=(Type=Bone,Name="spine_05"),Weight=0.2371),(Item=(Type=Bone,Name="spine_04"),Weight=0.5206))
    Weight=1
   )
   > &RigUnit_ParentConstraint(
    AdvancedSettings=(InterpolationType=Average,RotationOrderForFilter=XZY)
    Child=(Type=Null,Name="dyn_hair_bk_06_null")
    Parents="((Item=(Type=Bone,Name=\"spine_01\"),Weight=0.0165),(Item=(Type=Bone,Name=\"clavicle_l\"),Weight=0.0199),(Item=(Type=Bone,Name=\"clavicle_r\"),Weight=0.0199),(Item=(Type=Bone,Name=\"spine_02\"),Weight=0.2113),(Item=(Type=Bone,Name=\"spine_04\"),Weight=0.2299),(Item= ...(truncated)"
    Weight=1
   )
   > &RigUnit_ParentConstraint(
    AdvancedSettings=(InterpolationType=Average,RotationOrderForFilter=XZY)
    Child=(Type=Null,Name="dyn_hair_bk_r_01_null")
    Parents=((Item=(Type=Bone,Name="head"),Weight=1))
    Weight=1
   )
   > &RigUnit_ParentConstraint(
    AdvancedSettings=(InterpolationType=Average,RotationOrderForFilter=XZY)
    Child=(Type=Null,Name="dyn_hair_bk_r_02_null")
    Parents="((Item=(Type=Bone,Name=\"clavicle_r\"),Weight=0.0503),(Item=(Type=Bone,Name=\"spine_04\"),Weight=0.0613),(Item=(Type=Bone,Name=\"head\"),Weight=0.1054),(Item=(Type=Bone,Name=\"neck_02\"),Weight=0.1567),(Item=(Type=Bone,Name=\"spine_05\"),Weight=0.2204),(Item=(Type=B ...(truncated)"
    Weight=1
   )
   > &RigUnit_ParentConstraint(
    AdvancedSettings=(InterpolationType=Average,RotationOrderForFilter=XZY)
    Child=(Type=Null,Name="dyn_hair_bk_r_03_null")
    Parents=((Item=(Type=Bone,Name="clavicle_r"),Weight=0.4559),(Item=(Type=Bone,Name="spine_05"),Weight=0.5441))
    Weight=1
   )
   > &RigUnit_ParentConstraint(
    AdvancedSettings=(InterpolationType=Average,RotationOrderForFilter=XZY)
    Child=(Type=Null,Name="dyn_hair_bk_r_04_null")
    Parents=((Item=(Type=Bone,Name="upperarm_r"),Weight=0.0282),(Item=(Type=Bone,Name="spine_04"),Weight=0.0451),(Item=(Type=Bone,Name="spine_05"),Weight=0.377),(Item=(Type=Bone,Name="clavicle_r"),Weight=0.5498))
    Weight=1
   )
   > &RigUnit_ParentConstraint(
    AdvancedSettings=(InterpolationType=Average,RotationOrderForFilter=XZY)
    Child=(Type=Null,Name="dyn_hair_bk_r_05_null")
    Filter=(TranslationFilter=(bX=True,bY=True,bZ=True),RotationFilter=(bX=true,bY=true,bZ=true),ScaleFilter=(bX=True,bY=True,bZ=True))
    Parents=((Item=(Type=Bone,Name="upperarm_r"),Weight=0.0176),(Item=(Type=Bone,Name="spine_03"),Weight=0.0228),(Item=(Type=Bone,Name="spine_05"),Weight=0.285),(Item=(Type=Bone,Name="clavicle_l"),Weight=0.3331),(Item=(Type=Bone,Name="spine_04"),Weight=0.3415))
    Weight=1
   )
   > &RigUnit_ParentConstraint(
    AdvancedSettings=(InterpolationType=Average,RotationOrderForFilter=XZY)
    Child=(Type=Null,Name="dyn_hair_bk_r_06_null")
    Filter=(TranslationFilter=(bX=True,bY=True,bZ=True),RotationFilter=(bX=true,bY=true,bZ=true),ScaleFilter=(bX=True,bY=True,bZ=True))
    Parents=((Item=(Type=Bone,Name="spine_05"),Weight=0.0187),(Item=(Type=Bone,Name="spine_02"),Weight=0.0811),(Item=(Type=Bone,Name="clavicle_r"),Weight=0.1267),(Item=(Type=Bone,Name="spine_04"),Weight=0.3575),(Item=(Type=Bone,Name="spine_03"),Weight=0.416))
    Weight=1
   )
   > &RigUnit_ParentConstraint(
    AdvancedSettings=(InterpolationType=Average,RotationOrderForFilter=XZY)
    Child=(Type=Null,Name="dyn_hair_side_01_l_null")
    Parents=((Item=(Type=Bone,Name="spine_05"),Weight=0.2),(Item=(Type=Bone,Name="head"),Weight=0.6),(Item=(Type=Bone,Name="neck_01"),Weight=0.2))
    Weight=1
   )
   > &RigUnit_ParentConstraint(
    AdvancedSettings=(InterpolationType=Average,RotationOrderForFilter=XZY)
    Child=(Type=Null,Name="dyn_hair_side_02_l_null")
    Parents=((Item=(Type=Bone,Name="neck_01"),Weight=0.3),(Item=(Type=Bone,Name="spine_05"),Weight=0.3),(Item=(Type=Bone,Name="clavicle_l"),Weight=0.4))
    Weight=1
   )
   > &RigUnit_ParentConstraint(
    AdvancedSettings=(InterpolationType=Average,RotationOrderForFilter=XZY)
    Child=(Type=Null,Name="dyn_hair_side_03_l_null")
    Parents=((Item=(Type=Bone,Name="spine_05"),Weight=0.6),(Item=(Type=Bone,Name="clavicle_l"),Weight=0.4))
    Weight=1
   )
   > &RigUnit_ParentConstraint(
    AdvancedSettings=(InterpolationType=Average,RotationOrderForFilter=XZY)
    Child=(Type=Null,Name="dyn_hair_side_04_l_null")
    Parents=((Item=(Type=Bone,Name="spine_05"),Weight=1))
    Weight=1
   )
   > &RigUnit_ParentConstraint(
    AdvancedSettings=(InterpolationType=Average,RotationOrderForFilter=XZY)
    Child=(Type=Null,Name="dyn_hair_side_01_r_null")
    Parents=((Item=(Type=Bone,Name="spine_05"),Weight=0.2),(Item=(Type=Bone,Name="head"),Weight=0.6),(Item=(Type=Bone,Name="neck_01"),Weight=0.2))
    Weight=1
   )
   > &RigUnit_ParentConstraint(
    AdvancedSettings=(InterpolationType=Average,RotationOrderForFilter=XZY)
    Child=(Type=Null,Name="dyn_hair_side_02_r_null")
    Parents=((Item=(Type=Bone,Name="neck_01"),Weight=0.3),(Item=(Type=Bone,Name="spine_05"),Weight=0.3),(Item=(Type=Bone,Name="clavicle_r"),Weight=0.4))
    Weight=1
   )
   > &RigUnit_ParentConstraint(
    AdvancedSettings=(InterpolationType=Average,RotationOrderForFilter=XZY)
    Child=(Type=Null,Name="dyn_hair_side_03_r_null")
    Parents=((Item=(Type=Bone,Name="spine_05"),Weight=0.6),(Item=(Type=Bone,Name="clavicle_r"),Weight=0.4))
    Weight=1
   )
   > n2 = &RigUnit_ParentConstraint(
    AdvancedSettings=(InterpolationType=Average,RotationOrderForFilter=XZY)
    Child=(Type=Null,Name="dyn_hair_side_04_r_null")
    Parents=((Item=(Type=Bone,Name="spine_05"),Weight=1))
    Weight=1
   )
   RigVMFunctionReturnNode(ExecutePin=@n2.ExecutePin)
  ]}
 Multiply_10_ContainedGraph {nodes [
   n1 = RigVMFunctionEntryNode()
   RigVMFunctionReturnNode(Result=&RigVMFunction_MathDoubleMul(B=@n1.C, A=&RigVMFunction_MathDoubleMul(A=@n1.A, B=@n1.B).Result).Result)
  ]}
 Multiply_11_ContainedGraph {nodes [
   n1 = RigVMFunctionEntryNode()
   RigVMFunctionReturnNode(Result=&RigVMFunction_MathDoubleMul(B=@n1.C, A=&RigVMFunction_MathDoubleMul(A=@n1.A, B=@n1.B).Result).Result)
  ]}
 Multiply_14_ContainedGraph {nodes [
   n1 = RigVMFunctionEntryNode()
   RigVMFunctionReturnNode(Result=&RigVMFunction_MathDoubleMul(B=@n1.C, A=&RigVMFunction_MathDoubleMul(A=@n1.A, B=@n1.B).Result).Result)
  ]}
 Multiply_16_ContainedGraph {nodes [
   n1 = RigVMFunctionEntryNode()
   RigVMFunctionReturnNode(Result=&RigVMFunction_MathDoubleMul(B=@n1.C, A=&RigVMFunction_MathDoubleMul(A=@n1.A, B=@n1.B).Result).Result)
  ]}
 Multiply_16_ContainedGraph {nodes [
   n1 = RigVMFunctionEntryNode()
   RigVMFunctionReturnNode(Result=&RigVMFunction_MathDoubleMul(B=@n1.C, A=&RigVMFunction_MathDoubleMul(A=@n1.A, B=@n1.B).Result).Result)
  ]}
 Multiply_16_ContainedGraph {nodes [
   n1 = RigVMFunctionEntryNode()
   RigVMFunctionReturnNode(Result=&RigVMFunction_MathDoubleMul(B=@n1.C, A=&RigVMFunction_MathDoubleMul(A=@n1.A, B=@n1.B).Result).Result)
  ]}
 Multiply_17_ContainedGraph {nodes [
   n1 = RigVMFunctionEntryNode()
   RigVMFunctionReturnNode(Result=&RigVMFunction_MathDoubleMul(B=@n1.C, A=&RigVMFunction_MathDoubleMul(A=@n1.A, B=@n1.B).Result).Result)
  ]}
 Multiply_4_ContainedGraph {nodes [
   n1 = RigVMFunctionEntryNode()
   n2 = &RigVMFunction_MathDoubleMul(A=@n1.A, B=@n1.B)
   RigVMFunctionReturnNode(Result=&RigVMFunction_MathDoubleMul(B=@n1.D, A=&RigVMFunction_MathDoubleMul(B=@n1.C, A=@n2.Result).Result).Result)
  ]}
 New Function_ContainedGraph {nodes [
   n1 = RigVMFunctionEntryNode()
   RigVMFunction_MathBoolToInteger = &RigVMFunction_MathBoolToInteger(Value=@n1.onlyTwist)
   RigVMFunction_MathQuaternionSwingTwist = &RigVMFunction_MathQuaternionSwingTwist(
    twistAxis=@n1.twistAxis
    Input=&RigUnit_GetRelativeTransformForItem(bChildInitial=False, bParentInitial=False, child=@n1.child, parent=@n1.parent).RelativeTransform.Rotation
   )
   n2 = RigVMDispatchNode(
    Values=((),())
    Index=@RigVMFunction_MathBoolToInteger.Result
    Values.0=@RigVMFunction_MathQuaternionSwingTwist.Swing
    Values.1=@RigVMFunction_MathQuaternionSwingTwist.Twist
   ) {template="DISPATCH_RigVMDispatch_SelectInt32(in Index,in Values,out Result)"}
   RigVMFunction_MathQuaternionToEuler = &RigVMFunction_MathQuaternionToEuler(Value=@n2.Result, RotationOrder=@n1.Rotation)
   > DISPATCH_RigVMDispatch_Print = RigVMDispatchNode(ScreenColor=(R=1,G=1,B=1,A=1), ScreenDuration=1.5, Enabled=@n1.debugRotation, Value=@RigVMFunction_MathQuaternionToEuler.Result) {template="DISPATCH_RigVMDispatch_Print(in Prefix,in Value,in Enabled,in ScreenDuration,in ScreenColor)"}
   Cast = RigVMDispatchNode(Value=@n1.triggerAxis) {template="DISPATCH_RigVMDispatch_CastEnumToInt(in Value,out Result)"}
   DISPATCH_RigVMDispatch_SelectInt32 = RigVMDispatchNode(
    Values=(0,0,0)
    Index=@Cast.Result
    Values.0=@RigVMFunction_MathQuaternionToEuler.Result.X
    Values.1=@RigVMFunction_MathQuaternionToEuler.Result.Y
    Values.2=@RigVMFunction_MathQuaternionToEuler.Result.Z
   ) {template="DISPATCH_RigVMDispatch_SelectInt32(in Index,in Values,out Result)"}
   Remap = &RigVMFunction_MathDoubleRemap(
    bClamp=False
    Value=@DISPATCH_RigVMDispatch_SelectInt32.Result
    TargetMinimum=@n1.outputMinMax.X
    TargetMaximum=@n1.outputMinMax.Y
    SourceMinimum=@n1.rotationMinMax.X
    SourceMaximum=@n1.rotationMinMax.Y
   )
   Absolute = &RigVMFunction_MathDoubleAbs(Value=@Remap.Result)
   Multiply = &RigVMFunction_MathDoubleMul(
    A=&RigVMFunction_MathDoubleClamp(Value=RigVMDispatchNode(True=@Absolute.Result, Condition=@n1.absoluteValue, False=@Remap.Result) {template="DISPATCH_RigVMDispatch_If(in Condition,in True,in False,out Result)"}.Result).Result
    B=@n1.premult
   )
   > RigVMFunctionReturnNode(outputQuat=@n2.Result, Argument=@Multiply.Result)
  ]}
 New Function_ContainedGraph {nodes [
   n1 = RigVMFunctionEntryNode()
   RigVMFunction_ControlFlowBranch = &RigVMFunction_ControlFlowBranch(Condition=@n1.Control On, ExecuteContext=@n1.ExecuteContext)
   RigVMFunctionReturnNode(ExecuteContext=@RigVMFunction_ControlFlowBranch.Completed)
   RigUnit_GetTransform = &RigUnit_GetTransform(Item=(Type=Control,Name="neck_01_ctrl"), bInitial=False)
   RigUnit_SetTransform = &RigUnit_SetTransform(
    Item=(Type=Bone,Name="neck_01")
    Weight=1
    bInitial=False
    Value=@RigUnit_GetTransform.Transform
    ExecutePin=@RigVMFunction_ControlFlowBranch.True
   )
   > &RigUnit_SetTransform(
    Item=(Type=Bone,Name="neck_02")
    Weight=1
    bInitial=False
    Value=&RigUnit_GetTransform(Item=(Type=Control,Name="neck_02_ctrl"), bInitial=False).Transform
   )
   > &RigUnit_SetTransform(
    Item=(Type=Bone,Name="head")
    Weight=1
    bInitial=False
    Value=&RigUnit_GetTransform(Item=(Type=Control,Name="head_ctrl"), bInitial=False).Transform
   )
   > &RigUnit_SetTransform(
    Item=(Type=Bone,Name="clavicle_l")
    Weight=1
    bInitial=False
    Value=&RigUnit_GetTransform(Item=(Type=Control,Name="clavicle_l_ctrl"), bInitial=False).Transform
   )
   > &RigUnit_SetTransform(
    Item=(Type=Bone,Name="clavicle_r")
    Weight=1
    bInitial=False
    Value=&RigUnit_GetTransform(Item=(Type=Control,Name="clavicle_r_ctrl"), bInitial=False).Transform
   )
   > &RigUnit_SetTransform(
    Item=(Type=Bone,Name="upperarm_r")
    Weight=1
    bInitial=False
    Value=&RigUnit_GetTransform(Item=(Type=Control,Name="upperarm_r_ctrl"), bInitial=False).Transform
   )
  ]}
 New Function_ContainedGraph {nodes [
   RigVMFunctionEntryNode()
   > RigVMFunction_Sequence_1 = RigVMAggregateNode()
   DeltaRotation v4_46 = RigVMFunctionReferenceNode(
    Rotation=ZXY
    child=(Type=Bone,Name="head")
    outputMinMax=(X=0,Y=1)
    parent=(Type=Null,Name="Head_null")
    premult=0.333
    rotationMinMax=(X=0,Y=-40)
    triggerAxis=Z
    twistAxis=(X=1,Y=0,Z=0)
    ExecuteContext=@RigVMFunction_Sequence_1.A
   )
   > DeltaRotation v4_48 = RigVMFunctionReferenceNode(
    Rotation=XYZ
    child=(Type=Bone,Name="neck_02")
    outputMinMax=(X=0,Y=1)
    parent=(Type=Null,Name="Neck_02_null")
    premult=0.333
    rotationMinMax=(X=-5,Y=-12)
    triggerAxis=Z
    twistAxis=(X=1,Y=0,Z=0)
   )
   > DeltaRotation v4_47 = RigVMFunctionReferenceNode(
    Rotation=XYZ
    child=(Type=Bone,Name="neck_01")
    outputMinMax=(X=0,Y=1)
    parent=(Type=Null,Name="Neck_01_null")
    premult=0.333
    rotationMinMax=(X=0,Y=-30)
    triggerAxis=Z
    twistAxis=(X=1,Y=0,Z=0)
   )
   Add = RigVMAggregateNode(C=@"DeltaRotation v4_46".Argument, A=@"DeltaRotation v4_47".Argument, B=@"DeltaRotation v4_48".Argument)
   > RigVMVariableNode(Value=@Add.Result) {variable=Head Up}
   > DeltaRotation v4_49 = RigVMFunctionReferenceNode(
    Rotation=ZXY
    child=(Type=Bone,Name="head")
    onlyTwist=true
    outputMinMax=(X=0,Y=1)
    parent=(Type=Null,Name="Head_null")
    premult=0.8
    rotationMinMax=(X=0,Y=-40)
    triggerAxis=X
    twistAxis=(X=1,Y=0,Z=0)
   )
   > DeltaRotation v4_50 = RigVMFunctionReferenceNode(
    Rotation=XYZ
    child=(Type=Bone,Name="neck_01")
    onlyTwist=true
    outputMinMax=(X=0,Y=1)
    parent=(Type=Null,Name="Neck_01_null")
    premult=0.2
    rotationMinMax=(X=0,Y=-30)
    triggerAxis=X
    twistAxis=(X=1,Y=0,Z=0)
   )
   > RigVMVariableNode(Value=&RigVMFunction_MathDoubleAdd(B=@"DeltaRotation v4_49".Argument, A=@"DeltaRotation v4_50".Argument).Result) {variable=Head Left}
   > n1 = &RigUnit_SphericalPoseReader(
    ActiveRegionScaleFactors=(PositiveWidth=1,NegativeWidth=1,PositiveHeight=1,NegativeHeight=1)
    ActiveRegionSize=0.1
    Debug=(bDrawDebug=False,bDraw2D=False,bDrawLocalAxes=False,DebugScale=25,DebugSegments=20,DebugThickness=0.25)
    DriverAxis=(X=0,Y=1,Z=0)
    DriverItem=(Type=Bone,Name="spine_05")
    FalloffRegionScaleFactors=(PositiveWidth=1,NegativeWidth=1,PositiveHeight=1,NegativeHeight=1)
    FalloffSize=0.3
    FlipHeightScaling=False
    FlipWidthScaling=False
    OptionalParentItem=(Type=Bone,Name="pelvis")
    RotationOffset=(X=140,Y=0,Z=0)
   )
   > RigVMVariableNode(Value=@n1.OutputParam) {variable=Twist Spine Left}
   > DeltaRotation v4_51 = RigVMFunctionReferenceNode(
    Rotation=ZXY
    child=(Type=Bone,Name="head")
    outputMinMax=(X=0,Y=1)
    parent=(Type=Null,Name="Head_null")
    premult=0.5
    rotationMinMax=(X=0,Y=-13)
    triggerAxis=Y
    twistAxis=(X=1,Y=0,Z=0)
   )
   > DeltaRotation v4_55 = RigVMFunctionReferenceNode(
    Rotation=XYZ
    child=(Type=Bone,Name="neck_01")
    outputMinMax=(X=0,Y=1)
    parent=(Type=Null,Name="Neck_01_null")
    premult=0.5
    rotationMinMax=(X=0,Y=-10)
    triggerAxis=Y
    twistAxis=(X=1,Y=0,Z=0)
   )
   > RigVMVariableNode(Value=&RigVMFunction_MathDoubleAdd(B=@"DeltaRotation v4_51".Argument, A=@"DeltaRotation v4_55".Argument).Result) {variable=Head pitch Right}
   > DeltaRotation v4_56 = RigVMFunctionReferenceNode(
    Rotation=ZXY
    child=(Type=Bone,Name="clavicle_l")
    outputMinMax=(X=0,Y=1)
    parent=(Type=Null,Name="clavicle_l_null")
    premult=1
    rotationMinMax=(X=0,Y=-50)
    triggerAxis=Y
    twistAxis=(X=1,Y=0,Z=0)
   )
   > RigVMVariableNode(Value=@"DeltaRotation v4_56".Argument) {variable=Clavicle L Up}
   > n2 = &RigUnit_SphericalPoseReader(
    ActiveRegionScaleFactors=(PositiveWidth=1,NegativeWidth=1,PositiveHeight=1,NegativeHeight=1)
    ActiveRegionSize=0.1
    Debug=(bDrawDebug=False,bDraw2D=False,bDrawLocalAxes=False,DebugScale=25,DebugSegments=20,DebugThickness=0.25)
    DriverAxis=(X=0,Y=1,Z=0)
    DriverItem=(Type=Bone,Name="spine_05")
    FalloffRegionScaleFactors=(PositiveWidth=1,NegativeWidth=1,PositiveHeight=1,NegativeHeight=1)
    FalloffSize=0.3
    FlipHeightScaling=False
    FlipWidthScaling=False
    OptionalParentItem=(Type=Bone,Name="pelvis")
    RotationOffset=(X=20,Y=0,Z=0)
   )
   > RigVMVariableNode(Value=@n2.OutputParam) {variable=Twist Spline Right}
   > DeltaRotation v4_1 = RigVMFunctionReferenceNode(
    Rotation=ZXY
    child=(Type=Bone,Name="head")
    onlyTwist=true
    outputMinMax=(X=0,Y=1)
    parent=(Type=Null,Name="Head_null")
    premult=0.6
    rotationMinMax=(X=0,Y=40)
    triggerAxis=X
    twistAxis=(X=1,Y=0,Z=0)
   )
   > DeltaRotation v4_14 = RigVMFunctionReferenceNode(
    Rotation=XYZ
    child=(Type=Bone,Name="neck_01")
    onlyTwist=true
    outputMinMax=(X=0,Y=1)
    parent=(Type=Null,Name="Neck_01_null")
    premult=0.4
    rotationMinMax=(X=0,Y=30)
    triggerAxis=X
    twistAxis=(X=1,Y=0,Z=0)
   )
   > RigVMVariableNode(Value=&RigVMFunction_MathDoubleAdd(B=@"DeltaRotation v4_1".Argument, A=@"DeltaRotation v4_14".Argument).Result) {variable=Head Right}
   > DeltaRotation v4_2 = RigVMFunctionReferenceNode(
    Rotation=ZXY
    child=(Type=Bone,Name="clavicle_l")
    outputMinMax=(X=0,Y=1)
    parent=(Type=Null,Name="clavicle_l_null")
    premult=1
    rotationMinMax=(X=0,Y=-20)
    triggerAxis=Z
    twistAxis=(X=1,Y=0,Z=0)
   )
   > DeltaRotation v4_3 = RigVMFunctionReferenceNode(
    Rotation=ZXY
    child=(Type=Bone,Name="clavicle_r")
    outputMinMax=(X=0,Y=1)
    parent=(Type=Null,Name="clavicle_r_null")
    premult=1
    rotationMinMax=(X=0,Y=-20)
    triggerAxis=Z
    twistAxis=(X=1,Y=0,Z=0)
   )
   > RigVMVariableNode(Value=&RigVMFunction_MathDoubleMul(B=@"DeltaRotation v4_2".Argument, A=@"DeltaRotation v4_3".Argument).Result) {variable=Clavicle LR back}
   > DeltaRotation v4_10 = RigVMFunctionReferenceNode(
    Rotation=ZXY
    child=(Type=Bone,Name="head")
    outputMinMax=(X=0,Y=1)
    parent=(Type=Null,Name="Head_null")
    premult=0.4
    rotationMinMax=(X=0,Y=20)
    triggerAxis=Z
    twistAxis=(X=1,Y=0,Z=0)
   )
   > DeltaRotation v4_15 = RigVMFunctionReferenceNode(
    Rotation=XYZ
    child=(Type=Bone,Name="neck_02")
    outputMinMax=(X=0,Y=1)
    parent=(Type=Null,Name="Neck_02_null")
    premult=0.1
    rotationMinMax=(X=0,Y=12)
    triggerAxis=Z
    twistAxis=(X=1,Y=0,Z=0)
   )
   > DeltaRotation v4_11 = RigVMFunctionReferenceNode(
    Rotation=XYZ
    child=(Type=Bone,Name="neck_01")
    outputMinMax=(X=0,Y=1)
    parent=(Type=Null,Name="Neck_01_null")
    premult=0.5
    rotationMinMax=(X=0,Y=20)
    triggerAxis=Z
    twistAxis=(X=1,Y=0,Z=0)
   )
   Add_3 = RigVMAggregateNode(C=@"DeltaRotation v4_10".Argument, A=@"DeltaRotation v4_11".Argument, B=@"DeltaRotation v4_15".Argument)
   > RigVMVariableNode(Value=@Add_3.Result) {variable=Head Down}
   > DeltaRotation v4_17 = RigVMFunctionReferenceNode(
    Rotation=ZXY
    child=(Type=Bone,Name="clavicle_r")
    outputMinMax=(X=0,Y=1)
    parent=(Type=Null,Name="clavicle_r_null")
    premult=1
    rotationMinMax=(X=0,Y=-50)
    triggerAxis=Y
    twistAxis=(X=1,Y=0,Z=0)
   )
   > RigVMVariableNode(Value=@"DeltaRotation v4_17".Argument) {variable=Clavicle R Up}
   n3 = &RigUnit_GetRelativeTransformForItem(Child=(Type=Bone,Name="pelvis"), Parent=(Type=Bone,Name="root"), bChildInitial=False, bParentInitial=False)
   > RigVMVariableNode(Value=&RigVMFunction_MathDoubleMul(
     B=&RigVMFunction_MathDoubleRemap(SourceMaximum=55, SourceMinimum=46, TargetMaximum=0, TargetMinimum=1, Value=@n3.RelativeTransform.Translation.Z).Result
     A=&RigVMFunction_MathDoubleRemap(SourceMaximum=43, SourceMinimum=38, TargetMaximum=1, TargetMinimum=0, Value=@n3.RelativeTransform.Translation.Z).Result
    ).Result) {variable=Sit}
   > RigVMVariableNode(Value=&RigVMFunction_MathDoubleRemap(
     SourceMaximum=85
     SourceMinimum=80
     TargetMaximum=1
     TargetMinimum=0
     Value=&RigUnit_GetRelativeTransformForItem(Child=(Type=Bone,Name="pelvis"), Parent=(Type=Bone,Name="root"), bChildInitial=False, bParentInitial=False).RelativeTransform.Translation.Z
    ).Result) {variable=Stand}
   n4 = &RigUnit_GetRelativeTransformForItem(Child=(Type=Bone,Name="pelvis"), Parent=(Type=Bone,Name="root"), bChildInitial=False, bParentInitial=False)
   > RigVMVariableNode(Value=&RigVMFunction_MathDoubleMul(
     A=&RigVMFunction_MathDoubleRemap(SourceMaximum=70, SourceMinimum=60, TargetMaximum=1, TargetMinimum=0, Value=@n4.RelativeTransform.Translation.Z).Result
     B=&RigVMFunction_MathDoubleRemap(SourceMaximum=85, SourceMinimum=75, TargetMaximum=0, TargetMinimum=1, Value=@n4.RelativeTransform.Translation.Z).Result
    ).Result) {variable=Crouch}
   > DeltaRotation v4_12 = RigVMFunctionReferenceNode(
    Rotation=ZXY
    child=(Type=Bone,Name="clavicle_l")
    outputMinMax=(X=0,Y=1)
    parent=(Type=Null,Name="clavicle_l_null")
    premult=1
    rotationMinMax=(X=0,Y=-19)
    triggerAxis=Z
    twistAxis=(X=1,Y=0,Z=0)
   )
   > DeltaRotation v4_13 = RigVMFunctionReferenceNode(
    Rotation=ZXY
    child=(Type=Bone,Name="clavicle_l")
    outputMinMax=(X=0,Y=1)
    parent=(Type=Null,Name="clavicle_l_null")
    premult=1
    rotationMinMax=(X=-3,Y=-12)
    triggerAxis=Y
    twistAxis=(X=1,Y=0,Z=0)
   )
   Multiply_16 = RigVMAggregateNode(
    B=@"DeltaRotation v4_12".Argument
    A=@"DeltaRotation v4_13".Argument
    C=RigVMDispatchNode(False=1, True=0, Condition=RigVMVariableNode() {variable=Skydiving}.Value) {template="DISPATCH_RigVMDispatch_If(in Condition,in True,in False,out Result)"}.Result
   )
   > &RigUnit_ModifyTransforms(
    ItemToModify="((Item=(Type=Null,Name=\"dyn_hair_bk_l_03_null\"),Transform=(Rotation=(X=0,Y=0,Z=0,W=1),Translation=(X=0,Y=-3,Z=3),Scale3D=(X=1,Y=1,Z=1))),(Item=(Type=Null,Name=\"dyn_hair_bk_l_04_null\"),Transform=(Rotation=(X=0,Y=0,Z=0,W=1),Translation=(X=0,Y=-3,Z=2),Scale3D ...(truncated)"
    WeightMinimum=0
    Weight=@Multiply_16.Result
   )
   n5 = RigVMVariableNode() {variable=Head Up}
   &RigUnit_ModifyTransforms(
    ItemToModify="((Item=(Type=Null,Name=\"dyn_hair_bk_l_02_null\"),Transform=(Rotation=(X=0,Y=0,Z=0,W=1),Translation=(X=0,Y=-6,Z=0),Scale3D=(X=1,Y=1,Z=1))),(Item=(Type=Null,Name=\"dyn_hair_bk_l_03_null\"),Transform=(Rotation=(X=0,Y=0,Z=0,W=1),Translation=(X=0,Y=-6,Z=0),Scale3D ...(truncated)"
    WeightMinimum=0
    ExecutePin=@RigVMFunction_Sequence_1.B
    Weight=@n5.Value
   )
   > &RigUnit_ModifyTransforms(
    ItemToModify="((Item=(Type=Null,Name=\"dyn_hair_bk_02_null\"),Transform=(Rotation=(X=0,Y=0,Z=0,W=1),Translation=(X=0,Y=-6,Z=0),Scale3D=(X=1,Y=1,Z=1))),(Item=(Type=Null,Name=\"dyn_hair_bk_03_null\"),Transform=(Rotation=(X=0,Y=0,Z=0,W=1),Translation=(X=0,Y=-6,Z=0),Scale3D=(X= ...(truncated)"
    WeightMinimum=0
    Weight=@n5.Value
   )
   > &RigUnit_ModifyTransforms(
    ItemToModify="((Item=(Type=Null,Name=\"dyn_hair_bk_r_02_null\"),Transform=(Rotation=(X=0,Y=0,Z=0,W=1),Translation=(X=0,Y=-6,Z=0),Scale3D=(X=1,Y=1,Z=1))),(Item=(Type=Null,Name=\"dyn_hair_bk_r_03_null\"),Transform=(Rotation=(X=0,Y=0,Z=0,W=1),Translation=(X=0,Y=-6,Z=0),Scale3D ...(truncated)"
    WeightMinimum=0
    Weight=@n5.Value
   )
   n6 = RigVMVariableNode() {variable=Head Left}
   > &RigUnit_ModifyTransforms(
    ItemToModify="((Item=(Type=Null,Name=\"dyn_hair_bk_l_02_null\"),Transform=(Rotation=(X=0,Y=0,Z=0,W=1),Translation=(X=0,Y=0,Z=-4),Scale3D=(X=1,Y=1,Z=1))),(Item=(Type=Null,Name=\"dyn_hair_bk_l_03_null\"),Transform=(Rotation=(X=0,Y=0,Z=0,W=1),Translation=(X=0,Y=0,Z=-5),Scale3D ...(truncated)"
    WeightMinimum=0
    Weight=@n6.Value
   )
   > &RigUnit_ModifyTransforms(
    ItemToModify="((Item=(Type=Null,Name=\"dyn_hair_bk_02_null\"),Transform=(Rotation=(X=0,Y=0,Z=0,W=1),Translation=(X=0,Y=0,Z=-2),Scale3D=(X=1,Y=1,Z=1))),(Item=(Type=Null,Name=\"dyn_hair_bk_03_null\"),Transform=(Rotation=(X=0,Y=0,Z=0,W=1),Translation=(X=0,Y=0,Z=-3),Scale3D=(X= ...(truncated)"
    WeightMinimum=0
    Weight=@n6.Value
   )
   n7 = RigVMVariableNode() {variable=Twist Spine Left}
   > &RigUnit_ModifyTransforms(
    ItemToModify="((Item=(Type=Null,Name=\"dyn_hair_bk_l_02_null\"),Transform=(Rotation=(X=0,Y=0,Z=0,W=1),Translation=(X=0,Y=0,Z=-8),Scale3D=(X=1,Y=1,Z=1))),(Item=(Type=Null,Name=\"dyn_hair_bk_l_03_null\"),Transform=(Rotation=(X=0,Y=0,Z=0,W=1),Translation=(X=0,Y=-3,Z=-2),Scale3 ...(truncated)"
    WeightMinimum=0
    Weight=@n7.Value
   )
   > &RigUnit_ModifyTransforms(
    ItemToModify="((Item=(Type=Null,Name=\"dyn_hair_bk_02_null\"),Transform=(Rotation=(X=0,Y=0,Z=0,W=1),Translation=(X=0,Y=2,Z=-6),Scale3D=(X=1,Y=1,Z=1))),(Item=(Type=Null,Name=\"dyn_hair_bk_03_null\"),Transform=(Rotation=(X=0,Y=-0.0175,Z=0,W=0.9998),Translation=(X=0,Y=2,Z=-3), ...(truncated)"
    WeightMinimum=0
    Weight=@n7.Value
   )
   > &RigUnit_ModifyTransforms(
    ItemToModify="((Item=(Type=Null,Name=\"dyn_hair_bk_r_02_null\"),Transform=(Rotation=(X=0,Y=0,Z=0,W=1),Translation=(X=0,Y=2,Z=2),Scale3D=(X=1,Y=1,Z=1))),(Item=(Type=Null,Name=\"dyn_hair_bk_r_03_null\"),Transform=(Rotation=(X=0,Y=0,Z=0,W=1),Translation=(X=0,Y=3,Z=0),Scale3D=( ...(truncated)"
    WeightMinimum=0
    Weight=@n7.Value
   )
   Multiply_12 = RigVMAggregateNode(
    A=RigVMVariableNode() {variable=Head Up}.Value
    C=RigVMVariableNode() {variable=Clavicle L Up}.Value
    B=RigVMVariableNode() {variable=Head pitch Right}.Value
   )
   > &RigUnit_ModifyTransforms(
    ItemToModify="((Item=(Type=Null,Name=\"dyn_hair_bk_l_02_null\"),Transform=(Rotation=(X=0,Y=0,Z=0,W=1),Translation=(X=0,Y=0,Z=0),Scale3D=(X=1,Y=1,Z=1))),(Item=(Type=Null,Name=\"dyn_hair_bk_l_03_null\"),Transform=(Rotation=(X=0,Y=0,Z=0,W=1),Translation=(X=0,Y=0,Z=0),Scale3D=( ...(truncated)"
    WeightMinimum=0
    Weight=@Multiply_12.Result
   )
   > &RigUnit_ModifyTransforms(
    ItemToModify="((Item=(Type=Null,Name=\"dyn_hair_bk_02_null\"),Transform=(Rotation=(X=0,Y=0,Z=0,W=1),Translation=(X=0,Y=0,Z=0),Scale3D=(X=1,Y=1,Z=1))),(Item=(Type=Null,Name=\"dyn_hair_bk_03_null\"),Transform=(Rotation=(X=0,Y=0,Z=0,W=1),Translation=(X=0,Y=0,Z=0),Scale3D=(X=1, ...(truncated)"
    WeightMinimum=0
    Weight=@Multiply_12.Result
   )
   > &RigUnit_ModifyTransforms(
    ItemToModify="((Item=(Type=Null,Name=\"dyn_hair_bk_r_02_null\"),Transform=(Rotation=(X=0,Y=0,Z=0,W=1),Translation=(X=0,Y=0,Z=-5),Scale3D=(X=1,Y=1,Z=1))),(Item=(Type=Null,Name=\"dyn_hair_bk_r_03_null\"),Transform=(Rotation=(X=0,Y=0,Z=0,W=1),Translation=(X=0,Y=0,Z=-10),Scale3 ...(truncated)"
    WeightMinimum=0
    Weight=@Multiply_12.Result
   )
   n8 = RigVMVariableNode() {variable=Twist Spline Right}
   > &RigUnit_ModifyTransforms(
    ItemToModify="((Item=(Type=Null,Name=\"dyn_hair_bk_l_02_null\"),Transform=(Rotation=(X=0,Y=0,Z=0,W=1),Translation=(X=0,Y=0,Z=-2),Scale3D=(X=1,Y=1,Z=1))),(Item=(Type=Null,Name=\"dyn_hair_bk_l_03_null\"),Transform=(Rotation=(X=0,Y=0,Z=0,W=1),Translation=(X=0,Y=0,Z=-2),Scale3D ...(truncated)"
    WeightMinimum=0
    Weight=&RigVMFunction_MathDoubleRemap(SourceMaximum=0.7, SourceMinimum=0, TargetMaximum=1, TargetMinimum=0, Value=@n8.Value).Result
   )
   > &RigUnit_ModifyTransforms(
    ItemToModify="((Item=(Type=Null,Name=\"dyn_hair_bk_r_02_null\"),Transform=(Rotation=(X=0,Y=0,Z=0,W=1),Translation=(X=0,Y=0,Z=-5),Scale3D=(X=1,Y=1,Z=1))),(Item=(Type=Null,Name=\"dyn_hair_bk_r_03_null\"),Transform=(Rotation=(X=0,Y=0,Z=0,W=1),Translation=(X=0,Y=-4,Z=0),Scale3D ...(truncated)"
    WeightMinimum=0
    Weight=@n8.Value
   )
   Multiply_15 = RigVMAggregateNode(
    B=RigVMVariableNode() {variable=Clavicle L Up}.Value
    A=RigVMVariableNode() {variable=Head Up}.Value
    C=RigVMVariableNode() {variable=Crouch}.Value
   )
   > &RigUnit_ModifyTransforms(
    ItemToModify="((Item=(Type=Null,Name=\"dyn_hair_bk_l_02_null\"),Transform=(Rotation=(X=0,Y=0,Z=0,W=1),Translation=(X=0,Y=0,Z=0),Scale3D=(X=1,Y=1,Z=1))),(Item=(Type=Null,Name=\"dyn_hair_bk_l_03_null\"),Transform=(Rotation=(X=0,Y=0,Z=0,W=1),Translation=(X=0,Y=0,Z=0),Scale3D=( ...(truncated)"
    WeightMinimum=0
    Weight=@Multiply_15.Result
   )
   > &RigUnit_ModifyTransforms(
    ItemToModify="((Item=(Type=Null,Name=\"dyn_hair_bk_02_null\"),Transform=(Rotation=(X=0,Y=0,Z=0,W=1),Translation=(X=0,Y=0,Z=0),Scale3D=(X=1,Y=1,Z=1))),(Item=(Type=Null,Name=\"dyn_hair_bk_03_null\"),Transform=(Rotation=(X=0,Y=0,Z=0,W=1),Translation=(X=0,Y=0,Z=0),Scale3D=(X=1, ...(truncated)"
    WeightMinimum=0
    Weight=@Multiply_15.Result
   )
   > &RigUnit_ModifyTransforms(
    ItemToModify="((Item=(Type=Null,Name=\"dyn_hair_bk_r_02_null\"),Transform=(Rotation=(X=0,Y=0,Z=0,W=1),Translation=(X=0,Y=0,Z=0),Scale3D=(X=1,Y=1,Z=1))),(Item=(Type=Null,Name=\"dyn_hair_bk_r_03_null\"),Transform=(Rotation=(X=0,Y=0,Z=0,W=1),Translation=(X=0,Y=0,Z=0),Scale3D=( ...(truncated)"
    WeightMinimum=0
    Weight=@Multiply_15.Result
   )
   n9 = RigVMVariableNode() {variable=Head Right}
   > &RigUnit_ModifyTransforms(
    ItemToModify="((Item=(Type=Null,Name=\"dyn_hair_bk_l_02_null\"),Transform=(Rotation=(X=0,Y=0,Z=0,W=1),Translation=(X=0,Y=4,Z=6),Scale3D=(X=1,Y=1,Z=1))),(Item=(Type=Null,Name=\"dyn_hair_bk_l_03_null\"),Transform=(Rotation=(X=0,Y=0,Z=0,W=1),Translation=(X=0,Y=6,Z=4),Scale3D=( ...(truncated)"
    WeightMinimum=0
    Weight=@n9.Value
   )
   > &RigUnit_ModifyTransforms(
    ItemToModify="((Item=(Type=Null,Name=\"dyn_hair_bk_02_null\"),Transform=(Rotation=(X=0,Y=0,Z=0,W=1),Translation=(X=0,Y=0,Z=7),Scale3D=(X=1,Y=1,Z=1))),(Item=(Type=Null,Name=\"dyn_hair_bk_03_null\"),Transform=(Rotation=(X=0,Y=0,Z=0,W=1),Translation=(X=0,Y=0,Z=6),Scale3D=(X=1, ...(truncated)"
    WeightMinimum=0
    Weight=@n9.Value
   )
   > &RigUnit_ModifyTransforms(
    ItemToModify="((Item=(Type=Null,Name=\"dyn_hair_bk_r_02_null\"),Transform=(Rotation=(X=0,Y=0,Z=0,W=1),Translation=(X=0,Y=0,Z=12),Scale3D=(X=1,Y=1,Z=1))),(Item=(Type=Null,Name=\"dyn_hair_bk_r_03_null\"),Transform=(Rotation=(X=0,Y=0,Z=0,W=1),Translation=(X=0,Y=0,Z=10),Scale3D ...(truncated)"
    WeightMinimum=0
    Weight=@n9.Value
   )
   n10 = RigVMVariableNode() {variable=Clavicle LR back}
   > &RigUnit_ModifyTransforms(
    ItemToModify="((Item=(Type=Null,Name=\"dyn_hair_bk_l_02_null\"),Transform=(Rotation=(X=0,Y=0,Z=0,W=1),Translation=(X=0,Y=0,Z=2),Scale3D=(X=1,Y=1,Z=1))),(Item=(Type=Null,Name=\"dyn_hair_bk_l_03_null\"),Transform=(Rotation=(X=0,Y=0,Z=0,W=1),Translation=(X=0,Y=0,Z=2),Scale3D=( ...(truncated)"
    WeightMinimum=0
    Weight=@n10.Value
   )
   > &RigUnit_ModifyTransforms(
    ItemToModify="((Item=(Type=Null,Name=\"dyn_hair_bk_r_02_null\"),Transform=(Rotation=(X=0,Y=0,Z=0,W=1),Translation=(X=0,Y=0,Z=-2),Scale3D=(X=1,Y=1,Z=1))),(Item=(Type=Null,Name=\"dyn_hair_bk_r_03_null\"),Transform=(Rotation=(X=0,Y=0,Z=0,W=1),Translation=(X=0,Y=0,Z=-2),Scale3D ...(truncated)"
    WeightMinimum=0
    Weight=@n10.Value
   )
   n11 = RigVMVariableNode() {variable=Head Down}
   > &RigUnit_ModifyTransforms(
    ItemToModify="((Item=(Type=Null,Name=\"dyn_hair_bk_l_02_null\"),Transform=(Rotation=(X=0,Y=0,Z=0,W=1),Translation=(X=0,Y=3,Z=-1),Scale3D=(X=1,Y=1,Z=1))),(Item=(Type=Null,Name=\"dyn_hair_bk_l_03_null\"),Transform=(Rotation=(X=0,Y=0,Z=0,W=1),Translation=(X=0,Y=1,Z=-1),Scale3D ...(truncated)"
    WeightMinimum=0
    Weight=@n11.Value
   )
   > &RigUnit_ModifyTransforms(
    ItemToModify="((Item=(Type=Null,Name=\"dyn_hair_bk_02_null\"),Transform=(Rotation=(X=0,Y=0,Z=0,W=1),Translation=(X=0,Y=0,Z=0),Scale3D=(X=1,Y=1,Z=1))),(Item=(Type=Null,Name=\"dyn_hair_bk_03_null\"),Transform=(Rotation=(X=0,Y=0,Z=0,W=1),Translation=(X=0,Y=0,Z=0),Scale3D=(X=1, ...(truncated)"
    WeightMinimum=0
    Weight=@n11.Value
   )
   > &RigUnit_ModifyTransforms(
    ItemToModify="((Item=(Type=Null,Name=\"dyn_hair_bk_r_02_null\"),Transform=(Rotation=(X=0,Y=0,Z=0,W=1),Translation=(X=0,Y=4,Z=0),Scale3D=(X=1,Y=1,Z=1))),(Item=(Type=Null,Name=\"dyn_hair_bk_r_03_null\"),Transform=(Rotation=(X=0,Y=0,Z=0,W=1),Translation=(X=0,Y=0,Z=0),Scale3D=( ...(truncated)"
    WeightMinimum=0
    Weight=@n11.Value
   )
   > DeltaRotation v4_16 = RigVMFunctionReferenceNode(
    Rotation=ZXY
    child=(Type=Bone,Name="clavicle_r")
    outputMinMax=(X=0,Y=1)
    parent=(Type=Null,Name="clavicle_r_null")
    premult=1
    rotationMinMax=(X=0,Y=-9)
    triggerAxis=Z
    twistAxis=(X=1,Y=0,Z=0)
   )
   > &RigUnit_ModifyTransforms(
    ItemToModify="((Item=(Type=Null,Name=\"dyn_hair_bk_r_02_null\"),Transform=(Rotation=(X=0,Y=0,Z=0,W=1),Translation=(X=0,Y=0,Z=0),Scale3D=(X=1,Y=1,Z=1))),(Item=(Type=Null,Name=\"dyn_hair_bk_r_03_null\"),Transform=(Rotation=(X=0,Y=0,Z=0,W=1),Translation=(X=0,Y=-1,Z=0),Scale3D= ...(truncated)"
    WeightMinimum=0
    Weight=&RigVMFunction_MathDoubleMul(A=@"DeltaRotation v4_16".Argument, B=RigVMVariableNode() {variable=Stand}.Value).Result
   )
   Multiply_2 = RigVMAggregateNode(
    D=1
    A=RigVMVariableNode() {variable=Clavicle L Up}.Value
    B=RigVMVariableNode() {variable=Clavicle R Up}.Value
    C=RigVMVariableNode() {variable=Stand}.Value
   )
   > &RigUnit_ModifyTransforms(
    ItemToModify="((Item=(Type=Null,Name=\"dyn_hair_bk_l_02_null\"),Transform=(Rotation=(X=0,Y=0,Z=0,W=1),Translation=(X=0,Y=-2,Z=0),Scale3D=(X=1,Y=1,Z=1))),(Item=(Type=Null,Name=\"dyn_hair_bk_l_03_null\"),Transform=(Rotation=(X=0,Y=0,Z=0,W=1),Translation=(X=0,Y=-2,Z=2),Scale3D ...(truncated)"
    WeightMinimum=0
    Weight=@Multiply_2.Result
   )
   > &RigUnit_ModifyTransforms(
    ItemToModify="((Item=(Type=Null,Name=\"dyn_hair_bk_r_02_null\"),Transform=(Rotation=(X=0,Y=0,Z=0,W=1),Translation=(X=0,Y=-2,Z=0),Scale3D=(X=1,Y=1,Z=1))),(Item=(Type=Null,Name=\"dyn_hair_bk_r_03_null\"),Transform=(Rotation=(X=0,Y=0,Z=0,W=1),Translation=(X=0,Y=-2,Z=-3),Scale3 ...(truncated)"
    WeightMinimum=0
    Weight=@Multiply_2.Result
   )
   Remap = &RigVMFunction_MathDoubleRemap(SourceMaximum=0.4, SourceMinimum=0, TargetMaximum=1, TargetMinimum=0, Value=RigVMVariableNode() {variable=Head Up}.Value)
   Multiply_9 = RigVMAggregateNode(C=1, A=@Remap.Result, B=RigVMVariableNode() {variable=Sit}.Value)
   > &RigUnit_ModifyTransforms(
    ItemToModify="((Item=(Type=Null,Name=\"dyn_hair_bk_r_02_null\"),Transform=(Rotation=(X=0,Y=0,Z=0,W=1),Translation=(X=0,Y=0,Z=0),Scale3D=(X=1,Y=1,Z=1))),(Item=(Type=Null,Name=\"dyn_hair_bk_r_03_null\"),Transform=(Rotation=(X=0,Y=0,Z=0,W=1),Translation=(X=0,Y=-1,Z=0),Scale3D= ...(truncated)"
    WeightMinimum=0
    Weight=@Multiply_9.Result
   )
   Multiply_10 = RigVMAggregateNode(A=@Multiply_9.Result, C=RigVMVariableNode() {variable=Clavicle L Up}.Value, B=RigVMVariableNode() {variable=Head pitch Right}.Value)
   > &RigUnit_ModifyTransforms(
    ItemToModify="((Item=(Type=Null,Name=\"dyn_hair_bk_l_02_null\"),Transform=(Rotation=(X=0,Y=0,Z=0,W=1),Translation=(X=0,Y=0,Z=0),Scale3D=(X=1,Y=1,Z=1))),(Item=(Type=Null,Name=\"dyn_hair_bk_l_03_null\"),Transform=(Rotation=(X=0,Y=0,Z=0,W=1),Translation=(X=0,Y=0,Z=0),Scale3D=( ...(truncated)"
    WeightMinimum=0
    Weight=@Multiply_10.Result
   )
   > &RigUnit_ModifyTransforms(
    ItemToModify="((Item=(Type=Null,Name=\"dyn_hair_bk_02_null\"),Transform=(Rotation=(X=0,Y=0,Z=0,W=1),Translation=(X=0,Y=0,Z=-2),Scale3D=(X=1,Y=1,Z=1))),(Item=(Type=Null,Name=\"dyn_hair_bk_03_null\"),Transform=(Rotation=(X=0,Y=0,Z=0,W=1),Translation=(X=0,Y=0,Z=-3),Scale3D=(X= ...(truncated)"
    WeightMinimum=0
    Weight=@Multiply_10.Result
   )
   > &RigUnit_ModifyTransforms(
    ItemToModify="((Item=(Type=Null,Name=\"dyn_hair_bk_r_02_null\"),Transform=(Rotation=(X=0,Y=0,Z=0,W=1),Translation=(X=0,Y=0,Z=0),Scale3D=(X=1,Y=1,Z=1))),(Item=(Type=Null,Name=\"dyn_hair_bk_r_03_null\"),Transform=(Rotation=(X=0,Y=0,Z=0,W=1),Translation=(X=0,Y=0,Z=-4),Scale3D= ...(truncated)"
    WeightMinimum=0
    Weight=@Multiply_10.Result
   )
   > n12 = &RigUnit_SphericalPoseReader(
    ActiveRegionScaleFactors=(PositiveWidth=1,NegativeWidth=1,PositiveHeight=1,NegativeHeight=1)
    ActiveRegionSize=0.25
    Debug=(bDrawDebug=False,bDraw2D=False,bDrawLocalAxes=False,DebugScale=25,DebugSegments=20,DebugThickness=0.25)
    DriverAxis=(X=0,Y=1,Z=0)
    DriverItem=(Type=Bone,Name="head")
    FalloffRegionScaleFactors=(PositiveWidth=1,NegativeWidth=1,PositiveHeight=1,NegativeHeight=1)
    FalloffSize=0.45
    FlipHeightScaling=False
    FlipWidthScaling=False
    OptionalParentItem=(Type=Bone,Name="root")
    RotationOffset=(X=90,Y=0,Z=90)
   )
   Multiply_4 = RigVMAggregateNode(
    C=&RigVMFunction_MathDoubleRemap(SourceMaximum=0.5, SourceMinimum=0, TargetMaximum=1, TargetMinimum=0, Value=RigVMVariableNode() {variable=Twist Spline Right}.Value).Result
    B=@n12.OutputParam
    A=RigVMVariableNode() {variable=Sit}.Value
   )
   > &RigUnit_ModifyTransforms(
    ItemToModify="((Item=(Type=Null,Name=\"dyn_hair_bk_r_02_null\"),Transform=(Rotation=(X=0,Y=0,Z=0,W=1),Translation=(X=0,Y=-4,Z=1),Scale3D=(X=1,Y=1,Z=1))),(Item=(Type=Null,Name=\"dyn_hair_bk_r_03_null\"),Transform=(Rotation=(X=0,Y=0,Z=0,W=1),Translation=(X=-3,Y=-2,Z=-3),Scale ...(truncated)"
    WeightMinimum=0
    Weight=@Multiply_4.Result
   )
   Multiply_8 = RigVMAggregateNode(
    B=&RigVMFunction_MathDoubleRemap(SourceMaximum=0.5, SourceMinimum=0, TargetMaximum=1, TargetMinimum=0, Value=RigVMVariableNode() {variable=Head Up}.Value).Result
    C=&RigVMFunction_MathDoubleRemap(SourceMaximum=0.7, SourceMinimum=0, TargetMaximum=0, TargetMinimum=1, Value=RigVMVariableNode() {variable=Clavicle L Up}.Value).Result
    A=&RigVMFunction_MathDoubleRemap(SourceMaximum=0.5, SourceMinimum=0, TargetMaximum=1, TargetMinimum=0, Value=RigVMVariableNode() {variable=Head Left}.Value).Result
   )
   > &RigUnit_ModifyTransforms(
    ItemToModify="((Item=(Type=Null,Name=\"dyn_hair_bk_l_02_null\"),Transform=(Rotation=(X=0,Y=0,Z=0,W=1),Translation=(X=0,Y=0,Z=-4),Scale3D=(X=1,Y=1,Z=1))),(Item=(Type=Null,Name=\"dyn_hair_bk_l_03_null\"),Transform=(Rotation=(X=0,Y=0,Z=0,W=1),Translation=(X=0,Y=0,Z=-5),Scale3D ...(truncated)"
    WeightMinimum=0
    Weight=@Multiply_8.Result
   )
   n13 = RigVMVariableNode() {variable=Clavicle R Up}
   > n14 = &RigUnit_ModifyTransforms(
    ItemToModify="((Item=(Type=Null,Name=\"dyn_hair_bk_l_02_null\"),Transform=(Rotation=(X=0,Y=0,Z=0,W=1),Translation=(X=0,Y=0,Z=-4),Scale3D=(X=1,Y=1,Z=1))),(Item=(Type=Null,Name=\"dyn_hair_bk_l_03_null\"),Transform=(Rotation=(X=0,Y=0,Z=0,W=1),Translation=(X=0,Y=0,Z=-5),Scale3D ...(truncated)"
    WeightMinimum=0
    Weight=&RigVMFunction_MathDoubleMul(
     B=&RigVMFunction_MathDoubleRemap(SourceMaximum=1, SourceMinimum=0, TargetMaximum=1, TargetMinimum=0, Value=@n13.Value).Result
     A=@n12.OutputParam
    ).Result
   )
   n15 = &RigVMFunction_ControlFlowBranch(ExecuteContext=@n14.ExecutePin, Condition=RigVMVariableNode() {variable=Skydiving}.Value)
   n16 = &RigVMFunction_Sequence(ExecuteContext=@n15.True)
   n17 = &RigUnit_SphericalPoseReader(
    ActiveRegionScaleFactors=(PositiveWidth=1,NegativeWidth=1,PositiveHeight=1,NegativeHeight=1)
    ActiveRegionSize=0.2
    Debug=(bDrawDebug=True,bDraw2D=False,bDrawLocalAxes=False,DebugScale=25,DebugSegments=20,DebugThickness=0.25)
    DriverAxis=(X=0,Y=-1,Z=0)
    DriverItem=(Type=Bone,Name="spine_05")
    FalloffRegionScaleFactors=(PositiveWidth=1,NegativeWidth=1,PositiveHeight=1,NegativeHeight=1)
    FalloffSize=0.41
    FlipHeightScaling=False
    FlipWidthScaling=False
    OptionalParentItem=(Type=Bone,Name="spine_01")
    RotationOffset=(X=0,Y=90,Z=0)
    ExecutePin=@n16.A
   )
   > &RigUnit_ModifyTransforms(
    ItemToModify="((Item=(Type=Null,Name=\"dyn_hair_bk_l_03_null\"),Transform=(Rotation=(X=0,Y=0,Z=0,W=1),Translation=(X=0,Y=-5,Z=5),Scale3D=(X=1,Y=1,Z=1))),(Item=(Type=Null,Name=\"dyn_hair_bk_l_06_null\"),Transform=(Rotation=(X=0,Y=0,Z=0,W=1),Translation=(X=0,Y=-2,Z=0),Scale3D ...(truncated)"
    Weight=1
    WeightMinimum=0
   )
   > &RigUnit_ModifyTransforms(
    ItemToModify=((Item=(Type=Null,Name="dyn_hair_bk_06_null"),Transform=(Rotation=(X=0,Y=0,Z=0,W=1),Translation=(X=0,Y=-2,Z=0),Scale3D=(X=1,Y=1,Z=1))))
    WeightMinimum=0
    Weight=@n17.OutputParam
   )
   > &RigUnit_ModifyTransforms(
    ItemToModify=((Item=(Type=Null,Name="dyn_hair_bk_r_03_null"),Transform=(Rotation=(X=0,Y=0,Z=0,W=1),Translation=(X=0,Y=-3,Z=-2),Scale3D=(X=1,Y=1,Z=1))))
    WeightMinimum=0
    Weight=@n17.OutputParam
   )
   n18 = &RigVMFunction_MathDoubleMul(A=RigVMVariableNode() {variable=Clavicle L Up}.Value, B=RigVMVariableNode() {variable=Clavicle R Up}.Value)
   > &RigUnit_ModifyTransforms(
    ItemToModify="((Item=(Type=Null,Name=\"dyn_hair_bk_l_02_null\"),Transform=(Rotation=(X=0,Y=0,Z=0,W=1),Translation=(X=0,Y=-2,Z=0),Scale3D=(X=1,Y=1,Z=1))),(Item=(Type=Null,Name=\"dyn_hair_bk_l_03_null\"),Transform=(Rotation=(X=0,Y=0,Z=0,W=1),Translation=(X=0,Y=-2,Z=0),Scale3D ...(truncated)"
    WeightMinimum=0
    Weight=@n18.Result
   )
   > &RigUnit_ModifyTransforms(
    ItemToModify="((Item=(Type=Null,Name=\"dyn_hair_bk_r_02_null\"),Transform=(Rotation=(X=0,Y=0,Z=0,W=1),Translation=(X=0,Y=-2,Z=0),Scale3D=(X=1,Y=1,Z=1))),(Item=(Type=Null,Name=\"dyn_hair_bk_r_03_null\"),Transform=(Rotation=(X=0,Y=0,Z=0,W=1),Translation=(X=0,Y=-2,Z=0),Scale3D ...(truncated)"
    WeightMinimum=0
    Weight=@n18.Result
   )
   RigVMFunction_MathQuaternionToEuler = &RigVMFunction_MathQuaternionToEuler(
    RotationOrder=ZYX
    Value=&RigUnit_GetRelativeTransformForItem(Child=(Type=Bone,Name="spine_05"), Parent=(Type=Bone,Name="spine_01"), bChildInitial=False, bParentInitial=False).RelativeTransform.Rotation
   )
   n19 = &RigVMFunction_MathDoubleMul(
    B=RigVMDispatchNode(False=1, True=0, Condition=RigVMVariableNode() {variable=Pack}.Value) {template="DISPATCH_RigVMDispatch_If(in Condition,in True,in False,out Result)"}.Result
    A=&RigVMFunction_MathDoubleRemap(SourceMaximum=-56, SourceMinimum=-20, TargetMaximum=1, TargetMinimum=0, Value=@RigVMFunction_MathQuaternionToEuler.Result.Z).Result
   )
   &RigUnit_ModifyTransforms(
    ItemToModify="((Item=(Type=Null,Name=\"dyn_hair_bk_l_02_null\"),Transform=(Rotation=(X=0,Y=0,Z=0,W=1),Translation=(X=0,Y=0,Z=0),Scale3D=(X=1,Y=1,Z=1))),(Item=(Type=Null,Name=\"dyn_hair_bk_l_03_null\"),Transform=(Rotation=(X=0,Y=0,Z=0,W=1),Translation=(X=0,Y=3,Z=0),Scale3D=( ...(truncated)"
    WeightMinimum=0
    ExecutePin=@n16.B
    Weight=@n19.Result
   )
   > &RigUnit_ModifyTransforms(
    ItemToModify="((Item=(Type=Null,Name=\"dyn_hair_bk_02_null\"),Transform=(Rotation=(X=0,Y=0,Z=0,W=1),Translation=(X=0,Y=0,Z=0),Scale3D=(X=1,Y=1,Z=1))),(Item=(Type=Null,Name=\"dyn_hair_bk_03_null\"),Transform=(Rotation=(X=0,Y=0,Z=0,W=1),Translation=(X=0,Y=1,Z=0),Scale3D=(X=1, ...(truncated)"
    WeightMinimum=0
    Weight=@n19.Result
   )
   > &RigUnit_ModifyTransforms(
    ItemToModify="((Item=(Type=Null,Name=\"dyn_hair_bk_r_02_null\"),Transform=(Rotation=(X=0,Y=0,Z=0,W=1),Translation=(X=0,Y=0,Z=0),Scale3D=(X=1,Y=1,Z=1))),(Item=(Type=Null,Name=\"dyn_hair_bk_r_03_null\"),Transform=(Rotation=(X=0,Y=0,Z=0,W=1),Translation=(X=0,Y=3,Z=0),Scale3D=( ...(truncated)"
    WeightMinimum=0
    Weight=@n19.Result
   )
   n20 = &RigVMFunction_MathDoubleMul(
    B=RigVMDispatchNode(False=1, True=0, Condition=RigVMVariableNode() {variable=Pack}.Value) {template="DISPATCH_RigVMDispatch_If(in Condition,in True,in False,out Result)"}.Result
    A=RigVMVariableNode() {variable=Clavicle LR back}.Value
   )
   > &RigUnit_ModifyTransforms(
    ItemToModify="((Item=(Type=Null,Name=\"dyn_hair_bk_l_02_null\"),Transform=(Rotation=(X=0,Y=0,Z=0,W=1),Translation=(X=0,Y=0,Z=0),Scale3D=(X=1,Y=1,Z=1))),(Item=(Type=Null,Name=\"dyn_hair_bk_l_03_null\"),Transform=(Rotation=(X=0,Y=0,Z=0,W=1),Translation=(X=0,Y=-6,Z=0),Scale3D= ...(truncated)"
    WeightMinimum=0
    Weight=@n20.Result
   )
   > &RigUnit_ModifyTransforms(
    ItemToModify="((Item=(Type=Null,Name=\"dyn_hair_bk_02_null\"),Transform=(Rotation=(X=0,Y=0,Z=0,W=1),Translation=(X=0,Y=0,Z=0),Scale3D=(X=1,Y=1,Z=1))),(Item=(Type=Null,Name=\"dyn_hair_bk_03_null\"),Transform=(Rotation=(X=0,Y=0,Z=0,W=1),Translation=(X=0,Y=-2,Z=0),Scale3D=(X=1 ...(truncated)"
    WeightMinimum=0
    Weight=@n20.Result
   )
   > &RigUnit_ModifyTransforms(
    ItemToModify="((Item=(Type=Null,Name=\"dyn_hair_bk_r_02_null\"),Transform=(Rotation=(X=0,Y=0,Z=0,W=1),Translation=(X=0,Y=0,Z=0),Scale3D=(X=1,Y=1,Z=1))),(Item=(Type=Null,Name=\"dyn_hair_bk_r_03_null\"),Transform=(Rotation=(X=0,Y=0,Z=0,W=1),Translation=(X=0,Y=-6,Z=0),Scale3D= ...(truncated)"
    WeightMinimum=0
    Weight=@n20.Result
   )
   RigVMFunctionReturnNode(ExecuteContext=@RigVMFunction_Sequence_1.F)
   RigVMFunction_ControlFlowBranch = &RigVMFunction_ControlFlowBranch(Condition=RigVMVariableNode() {variable=RBAN}.Value, ExecuteContext=@RigVMFunction_Sequence_1.E)
   RigVMFunction_Sequence = &RigVMFunction_Sequence(ExecuteContext=@RigVMFunction_ControlFlowBranch.False)
   DeltaRotation v4_4 = RigVMFunctionReferenceNode(
    Rotation=ZXY
    child=(Type=Bone,Name="head")
    outputMinMax=(X=0,Y=1)
    parent=(Type=Null,Name="Head_null")
    premult=1
    rotationMinMax=(X=0,Y=-30)
    triggerAxis=Z
    twistAxis=(X=1,Y=0,Z=0)
    ExecuteContext=@RigVMFunction_Sequence.A
   )
   > &RigUnit_ModifyTransforms(
    ItemToModify="((Item=(Type=Null,Name=\"dyn_hair_side_04_l_null\"),Transform=(Rotation=(X=0,Y=0,Z=0,W=1),Translation=(X=0,Y=0,Z=-3),Scale3D=(X=1,Y=1,Z=1))),(Item=(Type=Null,Name=\"dyn_hair_side_03_l_null\"),Transform=(Rotation=(X=0,Y=0,Z=0,W=1),Translation=(X=0,Y=0,Z=-2),Sca ...(truncated)"
    WeightMinimum=0
    Weight=@"DeltaRotation v4_4".Argument
   )
   > DeltaRotation v4_6 = RigVMFunctionReferenceNode(
    Rotation=ZXY
    child=(Type=Bone,Name="clavicle_l")
    outputMinMax=(X=0,Y=1)
    parent=(Type=Null,Name="clavicle_l_null")
    premult=1
    rotationMinMax=(X=0,Y=20)
    triggerAxis=Z
    twistAxis=(X=1,Y=0,Z=0)
   )
   > &RigUnit_ModifyTransforms(
    ItemToModify=((Item=(Type=Null,Name="dyn_hair_side_03_l_null"),Transform=(Rotation=(X=0,Y=0,Z=0,W=1),Translation=(X=0,Y=-10,Z=0),Scale3D=(X=1,Y=1,Z=1))))
    WeightMinimum=0
    Weight=@"DeltaRotation v4_6".Argument
   )
   > DeltaRotation v4_5 = RigVMFunctionReferenceNode(
    Rotation=ZXY
    child=(Type=Bone,Name="clavicle_r")
    outputMinMax=(X=0,Y=1)
    parent=(Type=Null,Name="clavicle_r_null")
    premult=1
    rotationMinMax=(X=0,Y=20)
    triggerAxis=Z
    twistAxis=(X=1,Y=0,Z=0)
   )
   > &RigUnit_ModifyTransforms(
    ItemToModify=((Item=(Type=Null,Name="dyn_hair_side_03_r_null"),Transform=(Rotation=(X=0,Y=0,Z=0,W=1),Translation=(X=0,Y=10,Z=0),Scale3D=(X=1,Y=1,Z=1))))
    WeightMinimum=0
    Weight=@"DeltaRotation v4_5".Argument
   )
   > DeltaRotation v4_8 = RigVMFunctionReferenceNode(
    Rotation=ZXY
    child=(Type=Bone,Name="upperarm_r")
    outputMinMax=(X=0,Y=1)
    parent=(Type=Null,Name="upperarm_r_null")
    premult=1
    rotationMinMax=(X=60,Y=100)
    triggerAxis=Z
    twistAxis=(X=1,Y=0,Z=0)
   )
   > &RigUnit_ModifyTransforms(
    ItemToModify=((Item=(Type=Null,Name="dyn_hair_side_03_r_null"),Transform=(Rotation=(X=0,Y=0,Z=0,W=1),Translation=(X=0,Y=8,Z=0),Scale3D=(X=1,Y=1,Z=1))))
    WeightMinimum=0
    Weight=@"DeltaRotation v4_8".Argument
   )
   > DeltaRotation v4_7 = RigVMFunctionReferenceNode(
    Rotation=ZXY
    child=(Type=Bone,Name="head")
    outputMinMax=(X=0,Y=1)
    parent=(Type=Null,Name="Head_null")
    premult=1
    rotationMinMax=(X=0,Y=20)
    triggerAxis=Z
    twistAxis=(X=1,Y=0,Z=0)
   )
   > &RigUnit_ModifyTransforms(
    ItemToModify="((Item=(Type=Null,Name=\"dyn_hair_side_04_r_null\"),Transform=(Rotation=(X=0,Y=0,Z=0,W=1),Translation=(X=0,Y=2,Z=-2),Scale3D=(X=1,Y=1,Z=1))),(Item=(Type=Null,Name=\"dyn_hair_side_04_l_null\"),Transform=(Rotation=(X=0,Y=0,Z=0,W=1),Translation=(X=0,Y=-2,Z=-2),Sc ...(truncated)"
    WeightMinimum=0
    Weight=@"DeltaRotation v4_7".Argument
   )
   > DeltaRotation v4_9 = RigVMFunctionReferenceNode(
    child=(Type=Bone,Name="neck_01")
    onlyTwist=true
    outputMinMax=(X=0,Y=1)
    parent=(Type=Null,Name="Neck_01_null")
    premult=1
    rotationMinMax=(X=0,Y=-20)
    triggerAxis=X
    twistAxis=(X=1,Y=0,Z=0)
   )
   Multiply = &RigVMFunction_MathDoubleMul(B=@"DeltaRotation v4_7".Argument, A=@"DeltaRotation v4_9".Argument)
   > &RigUnit_ModifyTransforms(
    ItemToModify="((Item=(Type=Null,Name=\"dyn_hair_side_02_l_null\"),Transform=(Rotation=(X=0,Y=0,Z=0,W=1),Translation=(X=0,Y=4,Z=0),Scale3D=(X=1,Y=1,Z=1))),(Item=(Type=Null,Name=\"dyn_hair_side_01_l_null\"),Transform=(Rotation=(X=0,Y=0,Z=0,W=1),Translation=(X=0,Y=5,Z=0),Scale ...(truncated)"
    WeightMinimum=0
    Weight=@Multiply.Result
   )
   n21 = RigVMDispatchNode(Value=((Type=Bone,Name="dyn_hair_side_01_l"),(Type=Bone,Name="dyn_hair_side_02_l"),(Type=Bone,Name="dyn_hair_side_03_l"),(Type=Bone,Name="dyn_hair_side_04_l"))) {template="DISPATCH_RigVMDispatch_Constant(io Value)"}
   TwistBones_4 = RigVMFunctionReferenceNode(
    Axis to Twist=(bX=True,bY=False,bZ=False)
    Debug=false
    End=(Type=Bone,Name="spine_04")
    Start=(Type=Bone,Name="neck_01")
    Bones to Twist=@n21.Value
    ExecuteContext=@RigVMFunction_Sequence.B
   )
   n22 = RigVMDispatchNode(Value=((Type=Bone,Name="dyn_hair_side_01_r"),(Type=Bone,Name="dyn_hair_side_02_r"),(Type=Bone,Name="dyn_hair_side_03_r"),(Type=Bone,Name="dyn_hair_side_04_r"))) {template="DISPATCH_RigVMDispatch_Constant(io Value)"}
   > TwistBones_5 = RigVMFunctionReferenceNode(Axis to Twist=(bX=True,bY=False,bZ=False), End=(Type=Bone,Name="spine_04"), Start=(Type=Bone,Name="neck_01"), Bones to Twist=@n22.Value)
   n23 = RigVMDispatchNode(Value=((Type=Control,Name="dyn_hair_side_01_l_ctrl"),(Type=Control,Name="dyn_hair_side_02_l_ctrl"),(Type=Control,Name="dyn_hair_side_03_l_ctrl"),(Type=Control,Name="dyn_hair_side_04_l_ctrl"))) {template="DISPATCH_RigVMDispatch_Constant(io Value)"}
   n24 = RigVMDispatchNode(Value=((Type=Bone,Name="dyn_hair_side_01_l"),(Type=Bone,Name="dyn_hair_side_02_l"),(Type=Bone,Name="dyn_hair_side_03_l"),(Type=Bone,Name="dyn_hair_side_04_l"))) {template="DISPATCH_RigVMDispatch_Constant(io Value)"}
   > SplineIK_4 = RigVMFunctionReferenceNode(
    Primary Axis=(X=1,Y=0,Z=0)
    Samples Per Segment=1
    Secondary Spline Direction=(X=0,Y=0,Z=0)
    Spline Mode=BSpline
    Stretch=false
    Up Axis=(X=0,Y=0,Z=0)
    debug=false
    Controls=@n23.Value
    Bones=@n24.Value
   )
   n25 = RigVMDispatchNode(Value=((Type=Control,Name="dyn_hair_side_01_r_ctrl"),(Type=Control,Name="dyn_hair_side_02_r_ctrl"),(Type=Control,Name="dyn_hair_side_03_r_ctrl"),(Type=Control,Name="dyn_hair_side_04_r_ctrl"))) {template="DISPATCH_RigVMDispatch_Constant(io Value)"}
   n26 = RigVMDispatchNode(Value=((Type=Bone,Name="dyn_hair_side_01_r"),(Type=Bone,Name="dyn_hair_side_02_r"),(Type=Bone,Name="dyn_hair_side_03_r"),(Type=Bone,Name="dyn_hair_side_04_r"))) {template="DISPATCH_RigVMDispatch_Constant(io Value)"}
   > SplineIK_5 = RigVMFunctionReferenceNode(
    Primary Axis=(X=1,Y=0,Z=0)
    Samples Per Segment=1
    Secondary Spline Direction=(X=0,Y=0,Z=0)
    Spline Mode=BSpline
    Up Axis=(X=0,Y=0,Z=0)
    debug=false
    Controls=@n25.Value
    Bones=@n26.Value
   )
   &RigUnit_ModifyTransforms(
    ItemToModify="((Item=(Type=Bone,Name=\"dyn_hair_side_01_base_l\"),Transform=(Rotation=(X=-0.4009,Y=0.1509,Z=-0.1205,W=0.8955),Translation=(X=-4,Y=2,Z=0),Scale3D=(X=1,Y=1,Z=1))),(Item=(Type=Bone,Name=\"dyn_hair_side_01_base_r\"),Transform=(Rotation=(X=-0.6428,Y=0,Z=0,W=0.766 ...(truncated)"
    WeightMinimum=0
    ExecutePin=@RigVMFunction_ControlFlowBranch.True
    Weight=RigVMVariableNode() {variable=Head Left}.Value
   )
   > &RigUnit_ModifyTransforms(
    ItemToModify="((Item=(Type=Bone,Name=\"dyn_hair_side_01_base_r\"),Transform=(Rotation=(X=0.4009,Y=0.1509,Z=0.1205,W=0.8955),Translation=(X=-4,Y=-2,Z=0),Scale3D=(X=1,Y=1,Z=1))),(Item=(Type=Bone,Name=\"dyn_hair_side_01_base_l\"),Transform=(Rotation=(X=0.6428,Y=0,Z=0,W=0.766), ...(truncated)"
    WeightMinimum=0
    Weight=RigVMVariableNode() {variable=Head Right}.Value
   )
   > &RigUnit_ModifyTransforms(
    ItemToModify="((Item=(Type=Bone,Name=\"dyn_hair_side_01_base_r\"),Transform=(Rotation=(X=0,Y=0.0872,Z=0,W=0.9962),Translation=(X=0,Y=0,Z=0),Scale3D=(X=1,Y=1,Z=1))),(Item=(Type=Bone,Name=\"dyn_hair_side_01_base_l\"),Transform=(Rotation=(X=0,Y=0.1736,Z=0,W=0.9848),Translation ...(truncated)"
    WeightMinimum=0
    Weight=RigVMVariableNode() {variable=Head Down}.Value
   )
   n27 = RigVMVariableNode() {variable=Crouch}
   n28 = RigVMVariableNode() {variable=Head Up}
   &RigUnit_ModifyTransforms(
    ItemToModify=((Item=(Type=Bone,Name="corrective_back_hair"),Transform=(Rotation=(X=0,Y=0,Z=0,W=1),Translation=(X=10,Y=-8,Z=0),Scale3D=(X=1,Y=1,Z=1))))
    WeightMinimum=0
    ItemToModify.0.Transform.Translation.X=&RigVMFunction_MathDoubleRemap(SourceMaximum=1, SourceMinimum=0, TargetMaximum=10, TargetMinimum=20, Value=@n27.Value).Result
    ItemToModify.0.Transform.Translation.Y=&RigVMFunction_MathDoubleRemap(SourceMaximum=1, SourceMinimum=0, TargetMaximum=-8, TargetMinimum=0, Value=@n27.Value).Result
    Weight=@n28.Value
    ExecutePin=@RigVMFunction_Sequence_1.C
   )
   > &RigUnit_ModifyTransforms(
    ItemToModify=((Item=(Type=Bone,Name="corrective_back_hair"),Transform=(Rotation=(X=0.1736,Y=0,Z=0,W=0.9848),Translation=(X=5,Y=0,Z=0),Scale3D=(X=1,Y=1,Z=1))))
    WeightMinimum=0
    Weight=&RigVMFunction_MathDoubleSub(A=RigVMVariableNode() {variable=Clavicle R Up}.Value, B=@n28.Value).Result
   )
   Subtract = &RigVMFunction_MathDoubleSub(A=RigVMVariableNode() {variable=Clavicle L Up}.Value, B=@n28.Value)
   > &RigUnit_ModifyTransforms(
    ItemToModify=((Item=(Type=Bone,Name="corrective_back_hair"),Transform=(Rotation=(X=-0.1736,Y=0,Z=0,W=0.9848),Translation=(X=5,Y=0,Z=0),Scale3D=(X=1,Y=1,Z=1))))
    WeightMinimum=0
    Weight=@Subtract.Result
   )
   > &RigUnit_ModifyTransforms(
    ItemToModify=((Item=(Type=Bone,Name="corrective_back_hair"),Transform=(Rotation=(X=0.1736,Y=0,Z=0,W=0.9848),Translation=(X=2,Y=-10,Z=0),Scale3D=(X=1,Y=1,Z=1))))
    WeightMinimum=0
    Weight=&RigVMFunction_MathDoubleSub(A=RigVMVariableNode() {variable=Head Left}.Value, B=@n28.Value).Result
   )
   > &RigUnit_ModifyTransforms(
    ItemToModify=((Item=(Type=Bone,Name="corrective_back_hair"),Transform=(Rotation=(X=-0.1736,Y=0,Z=0,W=0.9848),Translation=(X=2,Y=0,Z=0),Scale3D=(X=1,Y=1,Z=1))))
    WeightMinimum=0
    Weight=&RigVMFunction_MathDoubleSub(A=RigVMVariableNode() {variable=Head Right}.Value, B=@n28.Value).Result
   )
   n29 = RigVMDispatchNode(Value=((Type=Bone,Name="dyn_hair_bk_l_02"),(Type=Bone,Name="dyn_hair_bk_l_03"),(Type=Bone,Name="dyn_hair_bk_l_04"),(Type=Bone,Name="dyn_hair_bk_l_05"),(Type=Bone,Name="dyn_hair_bk_l_06"))) {template="DISPATCH_RigVMDispatch_Constant(io Value)"}
   TwistBones_1 = RigVMFunctionReferenceNode(
    Axis to Twist=(bX=True,bY=False,bZ=False)
    End=(Type=Bone,Name="neck_01")
    Start=(Type=Bone,Name="spine_02")
    Bones to Twist=@n29.Value
    ExecuteContext=@RigVMFunction_Sequence_1.D
   )
   n30 = RigVMDispatchNode(Value=((Type=Bone,Name="dyn_hair_bk_02"),(Type=Bone,Name="dyn_hair_bk_03"),(Type=Bone,Name="dyn_hair_bk_04"),(Type=Bone,Name="dyn_hair_bk_05"),(Type=Bone,Name="dyn_hair_bk_06"))) {template="DISPATCH_RigVMDispatch_Constant(io Value)"}
   > TwistBones_2 = RigVMFunctionReferenceNode(Axis to Twist=(bX=True,bY=False,bZ=False), End=(Type=Bone,Name="neck_01"), Start=(Type=Bone,Name="spine_02"), Bones to Twist=@n30.Value)
   n31 = RigVMDispatchNode(Value=((Type=Bone,Name="dyn_hair_bk_r_02"),(Type=Bone,Name="dyn_hair_bk_r_03"),(Type=Bone,Name="dyn_hair_bk_r_04"),(Type=Bone,Name="dyn_hair_bk_r_05"),(Type=Bone,Name="dyn_hair_bk_r_06"))) {template="DISPATCH_RigVMDispatch_Constant(io Value)"}
   > TwistBones_3 = RigVMFunctionReferenceNode(Axis to Twist=(bX=True,bY=False,bZ=False), End=(Type=Bone,Name="neck_01"), Start=(Type=Bone,Name="spine_02"), Bones to Twist=@n31.Value)
   n32 = RigVMDispatchNode(Value="((Type=Control,Name=\"dyn_hair_bk_l_01_ctrl\"),(Type=Control,Name=\"dyn_hair_bk_l_02_ctrl\"),(Type=Control,Name=\"dyn_hair_bk_l_03_ctrl\"),(Type=Control,Name=\"dyn_hair_bk_l_04_ctrl\"),(Type=Control,Name=\"dyn_hair_bk_l_05_ctrl\"),(Type=Control,Name=\"dyn_hair_bk_l_0 ...(truncated)") {template="DISPATCH_RigVMDispatch_Constant(io Value)"}
   n33 = RigVMDispatchNode(Value=((Type=Bone,Name="dyn_hair_bk_l_01"),(Type=Bone,Name="dyn_hair_bk_l_02"),(Type=Bone,Name="dyn_hair_bk_l_03"),(Type=Bone,Name="dyn_hair_bk_l_04"),(Type=Bone,Name="dyn_hair_bk_l_05"),(Type=Bone,Name="dyn_hair_bk_l_06"))) {template="DISPATCH_RigVMDispatch_Constant(io Value)"}
   RigVMFunction_MathBoolAnd = &RigVMFunction_MathBoolAnd(A=RigVMVariableNode() {variable=Skydiving}.Value, B=RigVMVariableNode() {variable=Pack}.Value)
   > SplineIK_1 = RigVMFunctionReferenceNode(
    Primary Axis=(X=1,Y=0,Z=0)
    Samples Per Segment=1
    Secondary Spline Direction=(X=0,Y=0,Z=0)
    Spline Mode=BSpline
    Up Axis=(X=0,Y=0,Z=0)
    debug=false
    Controls=@n32.Value
    Bones=@n33.Value
    Stretch=@RigVMFunction_MathBoolAnd.Result
   )
   n34 = RigVMDispatchNode(Value=((Type=Bone,Name="dyn_hair_bk_01"),(Type=Bone,Name="dyn_hair_bk_02"),(Type=Bone,Name="dyn_hair_bk_03"),(Type=Bone,Name="dyn_hair_bk_04"),(Type=Bone,Name="dyn_hair_bk_05"),(Type=Bone,Name="dyn_hair_bk_06"))) {template="DISPATCH_RigVMDispatch_Constant(io Value)"}
   n35 = RigVMDispatchNode(Value=((Type=Control,Name="dyn_hair_bk_01_ctrl"),(Type=Control,Name="dyn_hair_bk_02_ctrl"),(Type=Control,Name="dyn_hair_bk_03_ctrl"),(Type=Control,Name="dyn_hair_bk_04_ctrl"),(Type=Control,Name="dyn_hair_bk_05_ctrl"),(Type=Control,Name="dyn_hair_bk_06_ctrl"))) {template="DISPATCH_RigVMDispatch_Constant(io Value)"}
   > SplineIK_2 = RigVMFunctionReferenceNode(
    Primary Axis=(X=1,Y=0,Z=0)
    Samples Per Segment=1
    Secondary Spline Direction=(X=0,Y=0,Z=0)
    Spline Mode=BSpline
    Up Axis=(X=0,Y=0,Z=0)
    debug=false
    Bones=@n34.Value
    Controls=@n35.Value
    Stretch=@RigVMFunction_MathBoolAnd.Result
   )
   n36 = RigVMDispatchNode(Value=((Type=Bone,Name="dyn_hair_bk_r_01"),(Type=Bone,Name="dyn_hair_bk_r_02"),(Type=Bone,Name="dyn_hair_bk_r_03"),(Type=Bone,Name="dyn_hair_bk_r_04"),(Type=Bone,Name="dyn_hair_bk_r_05"),(Type=Bone,Name="dyn_hair_bk_r_06"))) {template="DISPATCH_RigVMDispatch_Constant(io Value)"}
   n37 = RigVMDispatchNode(Value="((Type=Control,Name=\"dyn_hair_bk_r_01_ctrl\"),(Type=Control,Name=\"dyn_hair_bk_r_02_ctrl\"),(Type=Control,Name=\"dyn_hair_bk_r_03_ctrl\"),(Type=Control,Name=\"dyn_hair_bk_r_04_ctrl\"),(Type=Control,Name=\"dyn_hair_bk_r_05_ctrl\"),(Type=Control,Name=\"dyn_hair_bk_r_0 ...(truncated)") {template="DISPATCH_RigVMDispatch_Constant(io Value)"}
   > SplineIK_3 = RigVMFunctionReferenceNode(
    Primary Axis=(X=1,Y=0,Z=0)
    Samples Per Segment=1
    Secondary Spline Direction=(X=0,Y=0,Z=0)
    Spline Mode=BSpline
    Up Axis=(X=0,Y=0,Z=0)
    Bones=@n36.Value
    Controls=@n37.Value
    Stretch=@RigVMFunction_MathBoolAnd.Result
   )
   &RigUnit_ModifyTransforms(
    ExecutePin=()
    ItemToModify="((Item=(Type=Null,Name=\"dyn_hair_bk_r_02_null\"),Transform=(Rotation=(X=0,Y=0,Z=0,W=1),Translation=(X=0,Y=0,Z=0),Scale3D=(X=1,Y=1,Z=1))),(Item=(Type=Null,Name=\"dyn_hair_bk_r_03_null\"),Transform=(Rotation=(X=0,Y=0,Z=0,W=1),Translation=(X=0,Y=0,Z=0),Scale3D=( ...(truncated)"
    Weight=1
    WeightMinimum=0
   )
   &RigUnit_ModifyTransforms(
    ExecutePin=()
    ItemToModify="((Item=(Type=Bone,Name=\"dyn_hair_side_01_base_l\"),Transform=(Rotation=(X=0,Y=0,Z=0,W=1),Translation=(X=-6,Y=0,Z=0),Scale3D=(X=1,Y=1,Z=1))),(Item=(Type=Bone,Name=\"dyn_hair_side_01_base_r\"),Transform=(Rotation=(X=-0.6428,Y=0,Z=0,W=0.766),Translation=(X=0,Y=0 ...(truncated)"
    Weight=1
    WeightMinimum=0
   )
   &RigUnit_ModifyTransforms(
    ExecutePin=()
    ItemToModify="((Item=(Type=Bone,Name=\"dyn_hair_side_01_base_r\"),Transform=(Rotation=(X=0,Y=0,Z=0,W=1),Translation=(X=-6,Y=0,Z=0),Scale3D=(X=1,Y=1,Z=1))),(Item=(Type=Bone,Name=\"dyn_hair_side_01_base_l\"),Transform=(Rotation=(X=0.6428,Y=0,Z=0,W=0.766),Translation=(X=0,Y=0, ...(truncated)"
    Weight=1
    WeightMinimum=0
   )
   &RigUnit_ModifyTransforms(
    ExecutePin=()
    ItemToModify="((Item=(Type=Null,Name=\"dyn_hair_bk_l_02_null\"),Transform=(Rotation=(X=0,Y=0,Z=0,W=1),Translation=(X=0,Y=3,Z=0),Scale3D=(X=1,Y=1,Z=1))),(Item=(Type=Null,Name=\"dyn_hair_bk_l_03_null\"),Transform=(Rotation=(X=0,Y=0,Z=0,W=1),Translation=(X=0,Y=1,Z=0),Scale3D=( ...(truncated)"
    Weight=1
    WeightMinimum=0
   )
  ]}
 RigVMFunctionLibrary {nodes [
   Control Setup = RigVMCollapseNode(Control On=False, ExecuteContext=())
   DeltaRotation v4 = RigVMCollapseNode(
    ExecuteContext=()
    child=(Type=None,Name="None")
    outputMinMax=(X=0,Y=0)
    parent=(Type=None,Name="None")
    rotationMinMax=(X=0,Y=0)
    twistAxis=(X=0,Y=0,Z=0)
   )
   Hair Correction = RigVMCollapseNode(ExecuteContext=())
  ]}
 RigVMFunction_Sequence_3_ContainedGraph {nodes [
   n1 = RigVMFunctionEntryNode()
   n2 = &RigVMFunction_Sequence(ExecuteContext=@n1.ExecuteContext)
   n3 = &RigVMFunction_Sequence(ExecuteContext=@n2.B)
   n4 = &RigVMFunction_Sequence(ExecuteContext=@n3.B)
   n5 = &RigVMFunction_Sequence(ExecuteContext=@n4.B)
   n6 = &RigVMFunction_Sequence(ExecuteContext=@n5.B)
   RigVMFunctionReturnNode(A=@n2.A, B=@n3.A, C=@n4.A, D=@n5.A, E=@n6.A, F=@n6.B)
  ]}
 RigVMModel {nodes [
   RigUnit_BeginExecution = &RigUnit_BeginExecution()
   Control Setup = RigVMFunctionReferenceNode(Control On=false, ExecuteContext=@RigUnit_BeginExecution.ExecutePin)
   Constraints = RigVMCollapseNode(ExecutePin=@"Control Setup".ExecuteContext)
   Hair Correction = RigVMFunctionReferenceNode(ExecuteContext=@Constraints.ExecutePin)
  ]}
 RigVMModel ConstructionEvent Graph {nodes [
   n1 = &RigUnit_PrepareForExecution()
   Make_Array = RigVMDispatchNode(Values="((Type=Control,Name=\"dyn_hair_bk_01_ctrl\"),((),Type=Control,Name=\"dyn_hair_bk_02_ctrl\"),((),Type=Control,Name=\"dyn_hair_bk_03_ctrl\"),((),Type=Control,Name=\"dyn_hair_bk_04_ctrl\"),((),Type=Control,Name=\"dyn_hair_bk_05_ctrl\"),((),Type=Control,Name=\"dyn_hair_b ...(truncated)") {template="DISPATCH_RigVMDispatch_ArrayMake(in Values,out Array)"}
   RigVMVariableNode(Value=@Make_Array.Array, ExecuteContext=@n1.ExecutePin) {variable=dyn_Hair_Back_ctrls}
   > RigVMVariableNode(Value=/Script/ControlRig.RigUnit_HierarchyGetPoseItemArray(ElementType=All, Initial=false, ItemsToGet=@Make_Array.Array) {is=struct RigVMUnitNode}.Pose) {variable=dyn_Hair_Back_pose}
   RigUnit_GetTransform = &RigUnit_GetTransform(Item=(Type=Bone,Name="neck_01"), bInitial=False)
   > &RigUnit_HierarchyAddNull(Name=Neck_01_null, Parent=(Type=Bone,Name="spine_05"), Space=GlobalSpace, Transform=@RigUnit_GetTransform.Transform)
   > &RigUnit_HierarchyAddNull(
    Name=Neck_02_null
    Parent=(Type=Bone,Name="neck_01")
    Space=GlobalSpace
    Transform=&RigUnit_GetTransform(Item=(Type=Bone,Name="neck_02"), bInitial=False).Transform
   )
   > &RigUnit_HierarchyAddNull(
    Name=Head_null
    Parent=(Type=Bone,Name="neck_02")
    Space=GlobalSpace
    Transform=&RigUnit_GetTransform(Item=(Type=Bone,Name="head"), bInitial=False).Transform
   )
   > &RigUnit_HierarchyAddNull(
    Name=clavicle_l_null
    Parent=(Type=Bone,Name="spine_05")
    Space=GlobalSpace
    Transform=&RigUnit_GetTransform(Item=(Type=Bone,Name="clavicle_l"), bInitial=False).Transform
   )
   > &RigUnit_HierarchyAddNull(
    Name=clavicle_r_null
    Parent=(Type=Bone,Name="spine_05")
    Space=GlobalSpace
    Transform=&RigUnit_GetTransform(Item=(Type=Bone,Name="clavicle_r"), bInitial=False).Transform
   )
   > &RigUnit_HierarchyAddNull(
    Name=upperarm_r_null
    Parent=(Type=Bone,Name="clavicle_r")
    Space=GlobalSpace
    Transform=&RigUnit_GetTransform(Item=(Type=Bone,Name="upperarm_r"), bInitial=False).Transform
   )
  ]}
]
hierarchy [
 root {loc=<0 0 0>, rot=<0 0 0 1>}
 pelvis {parent=root, loc=<0 -0.5301 98.0571>, rot=<0.0127 0.707 -0.0127 -0.707>}
 spine_01 {parent=pelvis, loc=<6.1208 -0.9328 0>, rot=<0 0 0.034 -0.9994>}
 spine_02 {parent=spine_01, loc=<6.5997 -0.2095 0>, rot=<0 0 0.0238 -0.9997>}
 spine_03 {parent=spine_02, loc=<6.4914 -0.1237 0>, rot=<0 0 0.0149 -0.9999>}
 spine_04 {parent=spine_03, loc=<6.9598 0.0794 0>, rot=<0 0 -0.0132 -0.9999>}
 spine_05 {parent=spine_04, loc=<6.9987 0 0>, rot=<0 0 -0.0015 -1>}
 neck_01 {parent=spine_05, loc=<11.3138 0.1529 0>, rot=<0 0 -0.2003 -0.9797>}
 neck_02 {parent=neck_01, loc=<4.5627 0 0>, rot=<0 0 0.0504 -0.9987>}
 head {parent=neck_02, loc=<5.3914 0 0>, rot=<0 0 0.115 -0.9934>}
 dyn_hair_bk_r_base {parent=head, loc=<8.8381 -6.375 4.772>, rot=<0.1036 -0.9946 0 0>}
 dyn_hair_bk_r_01 {parent=dyn_hair_bk_r_base, loc=<0 0 0>, rot=<0 0 0 1>}
 dyn_hair_bk_r_02 {parent=dyn_hair_bk_r_01, loc=<9.1223 0 0>, rot=<0 0 0.0184 -0.9998>}
 dyn_hair_bk_r_03 {parent=dyn_hair_bk_r_02, loc=<8.4957 0.092 0>, rot=<0 0 0 1>}
 dyn_hair_bk_r_04 {parent=dyn_hair_bk_r_03, loc=<6 0 0>, rot=<0 0 -0.0809 -0.9967>}
 dyn_hair_bk_r_05 {parent=dyn_hair_bk_r_04, loc=<6 0 0>, rot=<0 0 -0.0902 -0.9959>}
 dyn_hair_bk_r_06 {parent=dyn_hair_bk_r_05, loc=<6 0 0>, rot=<0 0 0 1>}
 dyn_hair_bk_side_base_r {parent=dyn_hair_bk_r_02, loc=<5.1496 6.3978 -3.5091>, rot=<0.8011 0.5985 0 0>}
 dyn_hair_bk_side_01_r {parent=dyn_hair_bk_side_base_r, loc=<0 0 0>, rot=<0 0 0 1>}
 dyn_hair_bk_l_base {parent=head, loc=<8.8381 -6.375 -5.9285>, rot=<0.1036 -0.9946 0 0>}
 dyn_hair_bk_l_01 {parent=dyn_hair_bk_l_base, loc=<0 0 0>, rot=<0 0 0 1>}
 dyn_hair_bk_l_02 {parent=dyn_hair_bk_l_01, loc=<9.1223 0 0>, rot=<0 0 0.0184 -0.9998>}
 dyn_hair_bk_l_03 {parent=dyn_hair_bk_l_02, loc=<8.4957 0.092 0>, rot=<0 0 0 1>}
 dyn_hair_bk_l_04 {parent=dyn_hair_bk_l_03, loc=<6 0 0>, rot=<0 0 -0.0809 -0.9967>}
 dyn_hair_bk_l_05 {parent=dyn_hair_bk_l_04, loc=<6 0 0>, rot=<0 0 -0.0902 -0.9959>}
 dyn_hair_bk_l_06 {parent=dyn_hair_bk_l_05, loc=<6 0 0>, rot=<0 0 0 1>}
 dyn_hair_bk_side_base_l {parent=dyn_hair_bk_l_02, loc=<5.1495 6.3978 2.3526>, rot=<0.8011 0.5985 0 0>}
 dyn_hair_bk_side_01_l {parent=dyn_hair_bk_side_base_l, loc=<0 0 0>, rot=<0 0 0 1>}
 dyn_hair_bk_base {parent=head, loc=<9.0094 -7.1882 -0.6747>, rot=<0.1036 -0.9946 0 0>}
 dyn_hair_bk_01 {parent=dyn_hair_bk_base, loc=<0 0 0>, rot=<0 0 0 1>}
 dyn_hair_bk_02 {parent=dyn_hair_bk_01, loc=<9.1223 0 0>, rot=<0 0 0.0184 -0.9998>}
 dyn_hair_bk_03 {parent=dyn_hair_bk_02, loc=<8.4957 0.092 0>, rot=<0 0 0 1>}
 dyn_hair_bk_04 {parent=dyn_hair_bk_03, loc=<6 0 0>, rot=<0 0 -0.0809 -0.9967>}
 dyn_hair_bk_05 {parent=dyn_hair_bk_04, loc=<6 0 0>, rot=<0 0 -0.0902 -0.9959>}
 dyn_hair_bk_06 {parent=dyn_hair_bk_05, loc=<6 0 0>, rot=<0 0 0 1>}
 dyn_earring_base_l {parent=head, loc=<-1.136 0.1081 -6.1734>, rot=<0.0397 -0.5125 0.8574 0.0241>}
 dyn_earring_l {parent=dyn_earring_base_l, loc=<0 0 0>, rot=<0 0 0 1>}
 dyn_earring_base_r {parent=head, loc=<-1.1362 0.1081 6.1734>, rot=<-0.0241 -0.8574 0.5125 -0.0397>}
 dyn_earring_r {parent=dyn_earring_base_r, loc=<0 0 0>, rot=<0 0 0 1>}
 dyn_hair_side_01_base_r {parent=head, loc=<-1.5227 -0.9692 7.5161>, rot=<-0.0548 -0.7784 0.621 0.0736>}
 dyn_hair_side_01_r {parent=dyn_hair_side_01_base_r, loc=<0 0 0>, rot=<0 0 0 1>}
 dyn_hair_side_02_r {parent=dyn_hair_side_01_r, loc=<6.4502 0 0>, rot=<0.0012 -0.1683 -0.058 -0.984>}
 dyn_hair_side_03_r {parent=dyn_hair_side_02_r, loc=<9.8876 0 0>, rot=<0 -0.0021 -0.0183 -0.9998>}
 dyn_hair_side_04_r {parent=dyn_hair_side_03_r, loc=<7.1962 0 0>, rot=<0 0 0 1>}
 dyn_hair_side_01_base_l {parent=head, loc=<-1.5223 -0.9692 -7.5161>, rot=<-0.0736 -0.621 0.7784 0.0548>}
 dyn_hair_side_01_l {parent=dyn_hair_side_01_base_l, loc=<0 0 0>, rot=<0 0 0 1>}
 dyn_hair_side_02_l {parent=dyn_hair_side_01_l, loc=<6.4504 0 0>, rot=<-0.0012 -0.1683 0.058 -0.984>}
 dyn_hair_side_03_l {parent=dyn_hair_side_02_l, loc=<9.8879 0 0.0002>, rot=<0 -0.0021 0.0183 -0.9998>}
 dyn_hair_side_04_l {parent=dyn_hair_side_03_l, loc=<7.1959 0 -0.0002>, rot=<0 0 0 1>}
 corrective_back_hair {parent=neck_01, loc=<3.4322 -1.1491 0>, rot=<0 0 0.7071 -0.7071>}
 clavicle_l {parent=spine_05, loc=<5.4831 -0.1053 -3.2946>, rot=<0.0465 -0.7289 -0.009 -0.6829>}
 upperarm_l {parent=clavicle_l, loc=<9.8658 -1.2136 -0.4105>, rot=<0.0164 -0.3245 -0.0269 -0.9454>}
 lowerarm_l {parent=upperarm_l, loc=<26.465 -0.8345 0.259>, rot=<0.0036 0.0161 -0.2298 -0.9731>}
 hand_l {parent=lowerarm_l, loc=<22.7355 -0.3203 -0.6105>, rot=<-0.6784 -0.0223 0.0735 -0.7307>}
 clavicle_r {parent=spine_05, loc=<5.4831 -0.1053 3.2946>, rot=<0.7289 0.0465 -0.6829 0.009>}
 upperarm_r {parent=clavicle_r, loc=<-9.8658 1.2136 0.4105>, rot=<0.0164 -0.3245 -0.0269 -0.9454>}
 lowerarm_r {parent=upperarm_r, loc=<-26.465 0.8345 -0.259>, rot=<0.0036 0.0161 -0.2298 -0.9731>}
 hand_r {parent=lowerarm_r, loc=<-22.7355 0.3203 0.6106>, rot=<-0.6784 -0.0223 0.0735 -0.7307>}
 thigh_l {parent=pelvis, loc=<-1.0855 -0.2109 -8.7053>, rot=<0.0009 0.0355 0.0155 -0.9993>}
 calf_l {parent=thigh_l, loc=<-42.5284 -0.0019 -0.7297>, rot=<0.0006 -0.0161 -0.0391 -0.9991>}
 foot_l {parent=calf_l, loc=<-46.7401 -2.5014 -0.592>, rot=<0.0427 -0.0238 0.0353 -0.9982>}
 ball_l {parent=foot_l, loc=<-3.5658 10.8066 -0.1544>, rot=<0.0001 -0.0001 -0.7063 -0.7079>}
 thigh_r {parent=pelvis, loc=<-1.0855 -0.2109 8.7053>, rot=<0.0355 -0.0009 0.9993 0.0155>}
 calf_r {parent=thigh_r, loc=<42.5284 0.0019 0.7297>, rot=<0.0006 -0.0161 -0.0391 -0.9991>}
 foot_r {parent=calf_r, loc=<46.7401 2.5014 0.592>, rot=<0.0427 -0.0238 0.0353 -0.9982>}
 ball_r {parent=foot_r, loc=<3.5658 -10.8066 0.1544>, rot=<0.0001 -0.0001 -0.7063 -0.7079>}
 neck_01_ctrl {controlType=EULER_TRANSFORM, type=Control, loc=<0 0 0>, rot=<0 0 0 1>}
 neck_02_ctrl {controlType=EULER_TRANSFORM, parent=neck_01_ctrl, type=Control, loc=<0 0 0>, rot=<0 0 0 1>}
 head_ctrl {controlType=EULER_TRANSFORM, parent=neck_02_ctrl, type=Control, loc=<0 0 0>, rot=<0 0 0 1>}
 dyn_hair_bk_03_null {type=Null, loc=<0.6747 -12.6542 143.7608>, rot=<0.0833 -0.7022 0.0833 -0.7022>}
 dyn_hair_bk_03_ctrl {controlType=EULER_TRANSFORM, parent=dyn_hair_bk_03_null, type=Control, loc=<0 0 0>, rot=<0 0 0 1>}
 dyn_hair_bk_04_null {type=Null, loc=<0.6747 -14.058 137.9273>, rot=<-0.0262 0.7066 -0.0262 0.7066>}
 dyn_hair_bk_04_ctrl {controlType=EULER_TRANSFORM, parent=dyn_hair_bk_04_null, type=Control, loc=<0 0 0>, rot=<0 0 0 1>}
 dyn_hair_bk_05_null {type=Null, loc=<0.6747 -14.5031 131.9438>, rot=<-0.0376 -0.7061 -0.0376 -0.7061>}
 dyn_hair_bk_05_ctrl {controlType=EULER_TRANSFORM, parent=dyn_hair_bk_05_null, type=Control, loc=<0 0 0>, rot=<0 0 0 1>}
 dyn_hair_bk_06_null {type=Null, loc=<0.6747 -13.8658 125.9778>, rot=<0.0376 0.7061 0.0376 0.7061>}
 dyn_hair_bk_06_ctrl {controlType=EULER_TRANSFORM, parent=dyn_hair_bk_06_null, type=Control, loc=<0 0 0>, rot=<0 0 0 1>}
 dyn_hair_bk_r_03_null {type=Null, loc=<-4.772 -11.9224 143.608>, rot=<-0.0833 0.7022 -0.0833 0.7022>}
 dyn_hair_bk_r_03_ctrl {controlType=EULER_TRANSFORM, parent=dyn_hair_bk_r_03_null, type=Control, loc=<0 0 0>, rot=<0 0 0 1>}
 dyn_hair_bk_r_04_null {type=Null, loc=<-4.772 -13.3263 137.7746>, rot=<0.0262 -0.7066 0.0262 -0.7066>}
 dyn_hair_bk_r_04_ctrl {controlType=EULER_TRANSFORM, parent=dyn_hair_bk_r_04_null, type=Control, loc=<0 0 0>, rot=<0 0 0 1>}
 dyn_hair_bk_r_05_null {type=Null, loc=<-4.772 -13.7713 131.7921>, rot=<-0.0376 -0.7061 -0.0376 -0.7061>}
 dyn_hair_bk_r_05_ctrl {controlType=EULER_TRANSFORM, parent=dyn_hair_bk_r_05_null, type=Control, loc=<0 0 0>, rot=<0 0 0 1>}
 dyn_hair_bk_r_06_null {type=Null, loc=<-4.772 -13.134 125.826>, rot=<0.0376 0.7061 0.0376 0.7061>}
 dyn_hair_bk_r_06_ctrl {controlType=EULER_TRANSFORM, parent=dyn_hair_bk_r_06_null, type=Control, loc=<0 0 0>, rot=<0 0 0 1>}
 dyn_hair_bk_l_03_null {type=Null, loc=<5.9285 -11.9224 143.608>, rot=<-0.0833 0.7022 -0.0833 0.7022>}
 dyn_hair_bk_l_03_ctrl {controlType=EULER_TRANSFORM, parent=dyn_hair_bk_l_03_null, type=Control, loc=<0 0 0>, rot=<0 0 0 1>}
 dyn_hair_bk_l_04_null {type=Null, loc=<5.9285 -13.3263 137.7746>, rot=<0.0262 -0.7066 0.0262 -0.7066>}
 dyn_hair_bk_l_04_ctrl {controlType=EULER_TRANSFORM, parent=dyn_hair_bk_l_04_null, type=Control, loc=<0 0 0>, rot=<0 0 0 1>}
 dyn_hair_bk_l_05_null {type=Null, loc=<5.9285 -13.7713 131.7921>, rot=<-0.0376 -0.7061 -0.0376 -0.7061>}
 dyn_hair_bk_l_05_ctrl {controlType=EULER_TRANSFORM, parent=dyn_hair_bk_l_05_null, type=Control, loc=<0 0 0>, rot=<0 0 0 1>}
 dyn_hair_bk_l_06_null {type=Null, loc=<5.9285 -13.134 125.826>, rot=<0.0376 0.7061 0.0376 0.7061>}
 dyn_hair_bk_l_06_ctrl {controlType=EULER_TRANSFORM, parent=dyn_hair_bk_l_06_null, type=Control, loc=<0 0 0>, rot=<0 0 0 1>}
 clavicle_l_ctrl {controlType=EULER_TRANSFORM, type=Control, loc=<0 0 0>, rot=<0 0 0 1>}
 clavicle_r_ctrl {controlType=EULER_TRANSFORM, type=Control, loc=<0 0 0>, rot=<0 0 0 1>}
 dyn_hair_side_01_r_null {type=Null, loc=<-7.5161 -2.6439 150.5036>, rot=<-0.4799 -0.6041 0.3978 -0.4964>}
 dyn_hair_side_01_r_ctrl {controlType=EULER_TRANSFORM, parent=dyn_hair_side_01_r_null, type=Control, loc=<0 0 0>, rot=<0 0 0 1>}
 dyn_hair_side_02_r_null {type=Null, loc=<-7.8161 -1.451 144.1718>, rot=<0.5736 0.6507 -0.2812 0.4105>}
 dyn_hair_side_02_r_ctrl {controlType=EULER_TRANSFORM, parent=dyn_hair_side_02_r_null, type=Control, loc=<0 0 0>, rot=<0 0 0 1>}
 dyn_hair_side_03_r_null {type=Null, loc=<-7.8647 3.6473 135.7001>, rot=<-0.5861 -0.6409 0.2724 -0.4142>}
 dyn_hair_side_03_r_ctrl {controlType=EULER_TRANSFORM, parent=dyn_hair_side_03_r_null, type=Control, loc=<0 0 0>, rot=<0 0 0 1>}
 dyn_hair_side_04_r_null {type=Null, loc=<-7.6486 7.4295 129.5818>, rot=<0.5861 0.6409 -0.2724 0.4142>}
 dyn_hair_side_04_r_ctrl {controlType=EULER_TRANSFORM, parent=dyn_hair_side_04_r_null, type=Control, loc=<0 0 0>, rot=<0 0 0 1>}
 dyn_hair_side_01_l_null {type=Null, loc=<7.5161 -2.6439 150.504>, rot=<-0.6041 -0.4799 0.4964 -0.3978>}
 dyn_hair_side_01_l_ctrl {controlType=EULER_TRANSFORM, parent=dyn_hair_side_01_l_null, type=Control, loc=<0 0 0>, rot=<0 0 0 1>}
 dyn_hair_side_02_l_null {type=Null, loc=<7.8161 -1.451 144.172>, rot=<0.6507 0.5736 -0.4105 0.2812>}
 dyn_hair_side_02_l_ctrl {controlType=EULER_TRANSFORM, parent=dyn_hair_side_02_l_null, type=Control, loc=<0 0 0>, rot=<0 0 0 1>}
 dyn_hair_side_03_l_null {type=Null, loc=<7.8647 3.6473 135.7>, rot=<-0.6409 -0.5861 0.4142 -0.2724>}
 dyn_hair_side_03_l_ctrl {controlType=EULER_TRANSFORM, parent=dyn_hair_side_03_l_null, type=Control, loc=<0 0 0>, rot=<0 0 0 1>}
 dyn_hair_side_04_l_null {type=Null, loc=<7.6486 7.4295 129.582>, rot=<0.6409 0.5861 -0.4142 0.2724>}
 dyn_hair_side_04_l_ctrl {controlType=EULER_TRANSFORM, parent=dyn_hair_side_04_l_null, type=Control, loc=<0 0 0>, rot=<0 0 0 1>}
 upperarm_r_ctrl {controlType=EULER_TRANSFORM, type=Control, loc=<0 0 0>, rot=<0 0 0 1>}
 dyn_hair_bk_r_01_null {type=Null, loc=<-4.772 -8.1351 160.8194>, rot=<0.0703 -0.7036 0.0703 -0.7036>}
 dyn_hair_bk_r_01_ctrl {controlType=EULER_TRANSFORM, parent=dyn_hair_bk_r_01_null, type=Control, loc=<0 0 0>, rot=<0 0 0 1>}
 dyn_hair_bk_r_02_null {type=Null, loc=<-4.772 -9.9412 151.8776>, rot=<-0.0833 0.7022 -0.0833 0.7022>}
 dyn_hair_bk_r_02_ctrl {controlType=EULER_TRANSFORM, parent=dyn_hair_bk_r_02_null, type=Control, loc=<0 0 0>, rot=<0 0 0 1>}
 dyn_hair_bk_01_null {type=Null, loc=<0.6747 -8.9498 160.9839>, rot=<0.0703 -0.7036 0.0703 -0.7036>}
 dyn_hair_bk_01_ctrl {controlType=EULER_TRANSFORM, parent=dyn_hair_bk_01_null, type=Control, loc=<0 0 0>, rot=<0 0 0 1>}
 dyn_hair_bk_02_null {type=Null, loc=<0.6747 -10.7559 152.0421>, rot=<-0.0833 0.7022 -0.0833 0.7022>}
 dyn_hair_bk_02_ctrl {controlType=EULER_TRANSFORM, parent=dyn_hair_bk_02_null, type=Control, loc=<0 0 0>, rot=<0 0 0 1>}
 dyn_hair_bk_l_01_null {type=Null, loc=<5.9285 -8.1351 160.8194>, rot=<0.0703 -0.7036 0.0703 -0.7036>}
 dyn_hair_bk_l_01_ctrl {controlType=EULER_TRANSFORM, parent=dyn_hair_bk_l_01_null, type=Control, loc=<0 0 0>, rot=<0 0 0 1>}
 dyn_hair_bk_l_02_null {type=Null, loc=<5.9285 -9.9412 151.8776>, rot=<-0.0833 0.7022 -0.0833 0.7022>}
 dyn_hair_bk_l_02_ctrl {controlType=EULER_TRANSFORM, parent=dyn_hair_bk_l_02_null, type=Control, loc=<0 0 0>, rot=<0 0 0 1>}
]
variables [
 Clavicle L Up {type=double, default=0}
 Clavicle LR back {type=double, default=0}
 Clavicle R Up {type=double, default=0}
 Crouch {type=double, default=0}
 Head Down {type=double, default=0}
 Head Left {type=double, default=0}
 Head Right {type=double, default=0}
 Head Up {type=double, default=0}
 Head pitch Right {type=double, default=0}
 Pack {type=bool, default=False}
 RBAN {type=bool, default=False}
 Sit {type=double, default=0}
 Skydiving {type=bool, default=False}
 Stand {type=double, default=0}
 Twist Spine Left {type=double, default=0}
 Twist Spline Right {type=double, default=0}
 dyn_Hair_Back_ctrls {type="TArray<FRigElementKey>"}
 dyn_Hair_Back_pose {type=FRigPose, default=(Elements=,HierarchyTopologyVersion=-1,PoseHash=-1)}
]
