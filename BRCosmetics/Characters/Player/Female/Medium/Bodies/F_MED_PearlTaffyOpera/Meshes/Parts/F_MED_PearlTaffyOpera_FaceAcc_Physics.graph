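# Physics asset for the FaceAcc mesh of the PearlTaffyOpera female character, setting up left and right side-hair physics chains alongside upper-body collision shapes.
# Each chain runs from a small base body at the neck through four segments, changing from capsules to boxes toward the tip, with limited swing and twist per joint.
# Spine, neck, clavicle and arm bodies, including large angled boxes either side of the upper spine, use angularly free constraints, as do both chain bases.
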PhysicsAsset /BRCosmetics/Characters/Player/Female/Medium/Bodies/F_MED_PearlTaffyOpera/Meshes/Parts/F_MED_PearlTaffyOpera_FaceAcc_Physics
bodies {
 clavicle_l [{shapeType=Capsule, length=9, radius=5.005, shapeName=clavicle_l_capsule, center=<2.0958 -1.9208 0.3313>, rotation=<rot 90 0 0>}]
 clavicle_r [{shapeType=Capsule, length=9, radius=3.505, shapeName=clavicle_r_capsule, center=<-3.455 1.0394 -1.4426>, rotation=<rot 90 0 0>}]
 dyn_hair_side_01_base_l [{shapeType=Capsule, length=0.505, radius=0.505, shapeName=dyn_hair_side_01_base_l_capsule, rotation=<rot 90 0 0>}]
 dyn_hair_side_01_base_r [{shapeType=Capsule, length=0.505, radius=0.505, shapeName=dyn_hair_side_01_base_r_capsule, rotation=<rot 90 0 0>}]
 dyn_hair_side_01_l [{shapeType=Capsule, length=5.6971, radius=2.6604, shapeName=dyn_hair_side_01_l_capsule, center=<4.4976 0 0>, rotation=<rot 90 0 0>}]
 dyn_hair_side_01_r [{shapeType=Capsule, length=5.3894, radius=2.1514, shapeName=dyn_hair_side_01_r_capsule, center=<4.4976 0 0>, rotation=<rot 90 0 0>}]
 dyn_hair_side_02_l [{shapeType=Capsule, length=3, radius=3, shapeName=dyn_hair_side_02_l_capsule, center=<5.6868 -0.2094 0.1404>, rotation=<rot 90 0 0>}]
 dyn_hair_side_02_r [{shapeType=Capsule, length=3, radius=3, shapeName=dyn_hair_side_02_r_capsule, center=<5.6762 0 0>, rotation=<rot 90 0 0>}]
 dyn_hair_side_03_l [{shapeType=Box, extentX=4.5, extentY=7.5, extentZ=6.5, shapeName=dyn_hair_side_03_l_box, center=<7.1773 0 0>, rotation=<rot 0 0 0.7265>}]
 dyn_hair_side_03_r [{shapeType=Box, extentX=4.5, extentY=6.5, extentZ=6.5, shapeName=dyn_hair_side_03_r_box, center=<6.8711 0 0>}]
 dyn_hair_side_04_l [{
   shapeType=Box
   extentX=7
   extentY=7
   extentZ=10
   shapeName=dyn_hair_side_04_l_capsule
   center=<7.1655 0.2093 2.4336>
   rotation=<rot 70.232 179.9999 177.2099>
  }]
 dyn_hair_side_04_r [{
   shapeType=Box
   extentX=7
   extentY=7
   extentZ=11
   shapeName=dyn_hair_side_04_r_capsule
   center=<7.1655 0.2093 2.4336>
   rotation=<rot 70.232 179.9999 -2.7901>
  }]
 lowerarm_l [{shapeType=Capsule, length=21.5, radius=3.505, shapeName=lowerarm_l_capsule, center=<11.0679 0 0>, rotation=<rot -90 0 0>}]
 lowerarm_r [{shapeType=Capsule, length=21.5, radius=3.505, shapeName=lowerarm_r_capsule, center=<-11.0679 0 0>, rotation=<rot 90 0 0>}]
 neck_01 [
  {shapeType=Capsule, length=5.5, radius=4.005, shapeName=neck_01_capsule, center=<4.7514 -0.3825 0>, rotation=<rot 90 0 0>}
  {shapeType=Capsule, length=58.5, radius=9.5, shapeName=neck_01_capsule, center=<0 -16.9183 0>}
 ]
 spine_03 [{shapeType=Capsule, length=8, radius=9.505, shapeName=spine_03_capsule, center=<0 1.4852 0>, rotation=<rot 0 -90 0>}]
 spine_04 [{shapeType=Box, extentX=6, extentY=8, extentZ=22, shapeName=spine_04_box, center=<1.3409 5.9131 -0.2353>, rotation=<rot 0 -110.421 0>}]
 spine_05 [
  {shapeType=Box, extentX=13, extentY=7.5, extentZ=21.5, shapeName=spine_05_box, center=<-0.3244 2.2872 0>, rotation=<rot 0 -23.3619 0>}
  {
   shapeType=Box
   extentX=33
   extentY=30
   extentZ=13
   shapeName=spine_05_box
   center=<-0.0118 2.7878 -20.5594>
   rotation=<rot -1.3718 -9.0769 24.3978>
  }
  {shapeType=Box, extentX=33, extentY=30, extentZ=13, shapeName=spine_05_box, center=<0 3.9977 19.9557>, rotation=<rot 0 -0.0002 -21.6508>}
 ]
 upperarm_l [{shapeType=Capsule, length=31, radius=3.005, shapeName=upperarm_l_capsule, center=<14.0795 0 0>, rotation=<rot -90 0 0>}]
 upperarm_r [{shapeType=Capsule, length=30.5, radius=3.505, shapeName=upperarm_r_capsule, center=<-14.0795 0 0>, rotation=<rot 90 0 0>}]
}
constraints [
 {bone1Name=dyn_hair_side_01_l, bone2Name=dyn_hair_side_01_base_l, swing1LimitDegrees=10, swing2LimitDegrees=20, twistLimitDegrees=5}
 {bone1Name=dyn_hair_side_02_l, bone2Name=dyn_hair_side_01_l, swing1LimitDegrees=30, swing2LimitDegrees=30, twistLimitDegrees=15}
 {bone1Name=dyn_hair_side_03_l, bone2Name=dyn_hair_side_02_l, swing1LimitDegrees=35, swing2LimitDegrees=35, twistLimitDegrees=15}
 {bone1Name=dyn_hair_side_04_l, bone2Name=dyn_hair_side_03_l, swing1LimitDegrees=20, swing2LimitDegrees=35, twistLimitDegrees=15}
 {bone1Name=dyn_hair_side_01_r, bone2Name=dyn_hair_side_01_base_r, swing1LimitDegrees=10, swing2LimitDegrees=20, twistLimitDegrees=5}
 {bone1Name=dyn_hair_side_02_r, bone2Name=dyn_hair_side_01_r, swing1LimitDegrees=40, swing2LimitDegrees=40, twistLimitDegrees=8}
 {bone1Name=dyn_hair_side_03_r, bone2Name=dyn_hair_side_02_r, swing1LimitDegrees=35, swing2LimitDegrees=35, twistLimitDegrees=10}
 {bone1Name=dyn_hair_side_04_r, bone2Name=dyn_hair_side_03_r, swing1LimitDegrees=20, swing2LimitDegrees=35, twistLimitDegrees=8}
 {
  bone1Name=dyn_hair_side_01_base_r
  bone2Name=neck_01
  swing1LimitDegrees=45
  swing1Motion=Free
  swing2LimitDegrees=45
  swing2Motion=Free
  twistLimitDegrees=45
  twistMotion=Free
 }
 {
  bone1Name=dyn_hair_side_01_base_l
  bone2Name=neck_01
  swing1LimitDegrees=45
  swing1Motion=Free
  swing2LimitDegrees=45
  swing2Motion=Free
  twistLimitDegrees=45
  twistMotion=Free
 }
 {
  bone1Name=upperarm_l
  bone2Name=clavicle_l
  swing1LimitDegrees=45
  swing1Motion=Free
  swing2LimitDegrees=45
  swing2Motion=Free
  twistLimitDegrees=45
  twistMotion=Free
 }
 {
  bone1Name=upperarm_r
  bone2Name=clavicle_r
  swing1LimitDegrees=45
  swing1Motion=Free
  swing2LimitDegrees=45
  swing2Motion=Free
  twistLimitDegrees=45
  twistMotion=Free
 }
 {
  bone1Name=lowerarm_r
  bone2Name=upperarm_r
  swing1LimitDegrees=45
  swing1Motion=Free
  swing2LimitDegrees=45
  swing2Motion=Free
  twistLimitDegrees=45
  twistMotion=Free
 }
 {
  bone1Name=lowerarm_l
  bone2Name=upperarm_l
  swing1LimitDegrees=45
  swing1Motion=Free
  swing2LimitDegrees=45
  swing2Motion=Free
  twistLimitDegrees=45
  twistMotion=Free
 }
 {
  bone1Name=neck_01
  bone2Name=spine_05
  swing1LimitDegrees=45
  swing1Motion=Free
  swing2LimitDegrees=45
  swing2Motion=Free
  twistLimitDegrees=45
  twistMotion=Free
 }
 {
  bone1Name=clavicle_l
  bone2Name=spine_05
  swing1LimitDegrees=45
  swing1Motion=Free
  swing2LimitDegrees=45
  swing2Motion=Free
  twistLimitDegrees=45
  twistMotion=Free
 }
 {
  bone1Name=clavicle_r
  bone2Name=spine_05
  swing1LimitDegrees=45
  swing1Motion=Free
  swing2LimitDegrees=45
  swing2Motion=Free
  twistLimitDegrees=45
  twistMotion=Free
 }
 {
  bone1Name=spine_05
  bone2Name=spine_04
  swing1LimitDegrees=45
  swing1Motion=Free
  swing2LimitDegrees=45
  swing2Motion=Free
  twistLimitDegrees=45
  twistMotion=Free
 }
 {
  bone1Name=spine_04
  bone2Name=spine_03
  swing1LimitDegrees=45
  swing1Motion=Free
  swing2LimitDegrees=45
  swing2Motion=Free
  twistLimitDegrees=45
  twistMotion=Free
 }
]
