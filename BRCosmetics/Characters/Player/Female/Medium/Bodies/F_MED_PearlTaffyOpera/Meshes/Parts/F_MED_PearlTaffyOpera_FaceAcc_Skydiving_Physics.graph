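# Physics asset for the skydiving variant of the PearlTaffyOpera female face accessory, covering two side hair chains.
# Each chain runs from a base at neck_01 through four capsules with narrow swing limits and locked twist.
# Spine, neck, clavicle and arm capsules form the body shape, with large offset extra capsules on neck_01 and spine_05.
# Both chain tips also have angularly locked constraints to spine_04, while body joints rotate freely.
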PhysicsAsset /BRCosmetics/Characters/Player/Female/Medium/Bodies/F_MED_PearlTaffyOpera/Meshes/Parts/F_MED_PearlTaffyOpera_FaceAcc_Skydiving_Physics
bodies {
 clavicle_l [{shapeType=Capsule, length=10.5, radius=3.505, shapeName=clavicle_l_capsule, center=<2.0958 -1.3175 1.8881>, rotation=<rot 90 0 0>}]
 clavicle_r [{shapeType=Capsule, length=9, radius=3.505, shapeName=clavicle_r_capsule, center=<-3.455 1.2245 -1.4426>, rotation=<rot 90 0 0>}]
 dyn_hair_side_01_base_l [{shapeType=Capsule, length=0.505, radius=0.505, shapeName=dyn_hair_side_01_base_l_capsule, rotation=<rot 90 0 0>}]
 dyn_hair_side_01_base_r [{shapeType=Capsule, length=0.505, radius=0.505, shapeName=dyn_hair_side_01_base_r_capsule, rotation=<rot 90 0 0>}]
 dyn_hair_side_01_l [{shapeType=Capsule, length=5.6971, radius=2.6604, shapeName=dyn_hair_side_01_l_capsule, center=<4.4976 0 0>, rotation=<rot 90 0 0>}]
 dyn_hair_side_01_r [{shapeType=Capsule, length=5.3894, radius=2.1514, shapeName=dyn_hair_side_01_r_capsule, center=<4.4976 0 0>, rotation=<rot 90 0 0>}]
 dyn_hair_side_02_l [{shapeType=Capsule, length=3, radius=3, shapeName=dyn_hair_side_02_l_capsule, center=<5.6868 -0.2094 0.1404>, rotation=<rot 90 0 0>}]
 dyn_hair_side_02_r [{shapeType=Capsule, length=3, radius=3, shapeName=dyn_hair_side_02_r_capsule, center=<5.6762 0 0>, rotation=<rot 90 0 0>}]
 dyn_hair_side_03_l [{shapeType=Capsule, length=3, radius=3, shapeName=dyn_hair_side_03_l_capsule, center=<4.0102 -0.0777 0>, rotation=<rot 90 0 0>}]
 dyn_hair_side_03_r [{shapeType=Capsule, length=3, radius=3, shapeName=dyn_hair_side_03_r_capsule, center=<3.5437 0.2715 0.1613>, rotation=<rot 90 0 0>}]
 dyn_hair_side_04_l [{shapeType=Capsule, length=5.5, radius=4.5, shapeName=dyn_hair_side_04_l_capsule, center=<4.8734 0 0.7215>, rotation=<rot 73.2823 180 180>}]
 dyn_hair_side_04_r [{shapeType=Capsule, length=5.5, radius=4.5, shapeName=dyn_hair_side_04_r_capsule, center=<4.905 0 0.6118>, rotation=<rot 73.0602 -180 -180>}]
 lowerarm_l [{shapeType=Capsule, length=21.5, radius=3.505, shapeName=lowerarm_l_capsule, center=<11.0679 0 0>, rotation=<rot -90 0 0>}]
 lowerarm_r [{shapeType=Capsule, length=21.5, radius=3.505, shapeName=lowerarm_r_capsule, center=<-11.0679 0 0>, rotation=<rot 90 0 0>}]
 neck_01 [
  {shapeType=Capsule, length=5.5, radius=4.005, shapeName=neck_01_capsule, center=<4.7514 -0.3825 0>, rotation=<rot 90 0 0>}
  {shapeType=Capsule, length=55, radius=8, shapeName=neck_01_capsule, center=<0 -15.0258 0>}
 ]
 spine_03 [{shapeType=Capsule, length=11, radius=9.505, shapeName=spine_03_capsule, center=<-0.9449 1.5888 0>, rotation=<rot 0 -90 0>}]
 spine_04 [{shapeType=Capsule, length=12.505, radius=6.505, shapeName=spine_04_capsule, center=<-1.5094 3.8139 -0.2373>, rotation=<rot 0 -90 0>}]
 spine_05 [
  {shapeType=Capsule, length=11.005, radius=6.505, shapeName=spine_05_capsule, center=<-1.1737 -0.0657 0>, rotation=<rot 0 -90 0>}
  {shapeType=Capsule, length=26.5, radius=15, shapeName=spine_05_capsule, center=<0 -15.1303 11.3915>, rotation=<rot 0 0 15.2372>}
 ]
 upperarm_l [{shapeType=Capsule, length=31, radius=3.005, shapeName=upperarm_l_capsule, center=<14.0795 0 0>, rotation=<rot -90 0 0>}]
 upperarm_r [{shapeType=Capsule, length=30.5, radius=3.505, shapeName=upperarm_r_capsule, center=<-14.0795 0 0>, rotation=<rot 90 0 0>}]
}
constraints [
 {
  bone1Name=dyn_hair_side_01_l
  bone2Name=dyn_hair_side_01_base_l
  swing1LimitDegrees=5
  swing2LimitDegrees=20
  twistLimitDegrees=5
  twistMotion=Locked
 }
 {
  bone1Name=dyn_hair_side_02_l
  bone2Name=dyn_hair_side_01_l
  swing1LimitDegrees=5
  swing2LimitDegrees=20
  twistLimitDegrees=8
  twistMotion=Locked
 }
 {
  bone1Name=dyn_hair_side_03_l
  bone2Name=dyn_hair_side_02_l
  swing1LimitDegrees=5
  swing2LimitDegrees=10
  twistLimitDegrees=10
  twistMotion=Locked
 }
 {
  bone1Name=dyn_hair_side_04_l
  bone2Name=dyn_hair_side_03_l
  swing1LimitDegrees=5
  swing2LimitDegrees=10
  twistLimitDegrees=8
  twistMotion=Locked
 }
 {
  bone1Name=dyn_hair_side_01_r
  bone2Name=dyn_hair_side_01_base_r
  swing1LimitDegrees=5
  swing2LimitDegrees=20
  twistLimitDegrees=5
  twistMotion=Locked
 }
 {
  bone1Name=dyn_hair_side_02_r
  bone2Name=dyn_hair_side_01_r
  swing1LimitDegrees=5
  swing2LimitDegrees=20
  twistLimitDegrees=8
  twistMotion=Locked
 }
 {
  bone1Name=dyn_hair_side_03_r
  bone2Name=dyn_hair_side_02_r
  swing1LimitDegrees=5
  swing2LimitDegrees=10
  twistLimitDegrees=10
  twistMotion=Locked
 }
 {
  bone1Name=dyn_hair_side_04_r
  bone2Name=dyn_hair_side_03_r
  swing1LimitDegrees=5
  swing2LimitDegrees=10
  twistLimitDegrees=8
  twistMotion=Locked
 }
 {
  bone1Name=dyn_hair_side_01_base_r
  bone2Name=neck_01
  swing1LimitDegrees=45
  swing1Motion=Free
  swing2LimitDegrees=45
  swing2Motion=Free
  twistLimitDegrees=45
  twistMotion=Free
 }
 {
  bone1Name=dyn_hair_side_01_base_l
  bone2Name=neck_01
  swing1LimitDegrees=45
  swing1Motion=Free
  swing2LimitDegrees=45
  swing2Motion=Free
  twistLimitDegrees=45
  twistMotion=Free
 }
 {
  bone1Name=upperarm_l
  bone2Name=clavicle_l
  swing1LimitDegrees=45
  swing1Motion=Free
  swing2LimitDegrees=45
  swing2Motion=Free
  twistLimitDegrees=45
  twistMotion=Free
 }
 {
  bone1Name=upperarm_r
  bone2Name=clavicle_r
  swing1LimitDegrees=45
  swing1Motion=Free
  swing2LimitDegrees=45
  swing2Motion=Free
  twistLimitDegrees=45
  twistMotion=Free
 }
 {
  bone1Name=lowerarm_r
  bone2Name=upperarm_r
  swing1LimitDegrees=45
  swing1Motion=Free
  swing2LimitDegrees=45
  swing2Motion=Free
  twistLimitDegrees=45
  twistMotion=Free
 }
 {
  bone1Name=lowerarm_l
  bone2Name=upperarm_l
  swing1LimitDegrees=45
  swing1Motion=Free
  swing2LimitDegrees=45
  swing2Motion=Free
  twistLimitDegrees=45
  twistMotion=Free
 }
 {
  bone1Name=neck_01
  bone2Name=spine_05
  swing1LimitDegrees=45
  swing1Motion=Free
  swing2LimitDegrees=45
  swing2Motion=Free
  twistLimitDegrees=45
  twistMotion=Free
 }
 {
  bone1Name=clavicle_l
  bone2Name=spine_05
  swing1LimitDegrees=45
  swing1Motion=Free
  swing2LimitDegrees=45
  swing2Motion=Free
  twistLimitDegrees=45
  twistMotion=Free
 }
 {
  bone1Name=clavicle_r
  bone2Name=spine_05
  swing1LimitDegrees=45
  swing1Motion=Free
  swing2LimitDegrees=45
  swing2Motion=Free
  twistLimitDegrees=45
  twistMotion=Free
 }
 {
  bone1Name=spine_05
  bone2Name=spine_04
  swing1LimitDegrees=45
  swing1Motion=Free
  swing2LimitDegrees=45
  swing2Motion=Free
  twistLimitDegrees=45
  twistMotion=Free
 }
 {
  bone1Name=spine_04
  bone2Name=spine_03
  swing1LimitDegrees=45
  swing1Motion=Free
  swing2LimitDegrees=45
  swing2Motion=Free
  twistLimitDegrees=45
  twistMotion=Free
 }
 {
  bone1Name=spine_04
  bone2Name=dyn_hair_side_04_r
  swing1LimitDegrees=45
  swing1Motion=Locked
  swing2LimitDegrees=45
  swing2Motion=Locked
  twistLimitDegrees=45
  twistMotion=Locked
 }
 {
  bone1Name=spine_04
  bone2Name=dyn_hair_side_04_l
  swing1LimitDegrees=45
  swing1Motion=Locked
  swing2LimitDegrees=45
  swing2Motion=Locked
  twistLimitDegrees=45
  twistMotion=Locked
 }
]
